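# Material function for the Cyclone Universe character body that blends a scrolling planet surface over incoming material attributes by an Alpha input.
# Planet diffuse, normal and spec textures share a UV1 panner driven by PanX/PanY parameters, with spec red driving both roughness and specular.
# The planet attributes pass through the reused weapon-skin GenericGem function, masked by inverted spec times a squared, inverted vertical UV gradient, before the final blend.
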MaterialFunction /BRCosmetics/Characters/Player/Male/Medium/Bodies/M_MED_CycloneUniverse/Materials/MF_Cyclone_RotatingPlanet
graph {nodes [
  n1 = TextureCoordinate() {coordinateIndex=1}
  n2 = Panner(Coordinate=@n1, Speed=AppendVector(A=ScalarParameter() {defaultValue=0.03, parameterName=PanX}, B=ScalarParameter() {parameterName=PanY}))
  n3 = TextureSample(UVs=@n2) {texture=/BRCosmetics/Characters/Player/Male/Medium/Bodies/M_MED_CycloneUniverse/Textures/Planet_Spec}
  n4 = OneMinus(None=Saturate(None=MaterialFunctionCall(UV Channel=@n1) {materialFunction=&LinearGradient}.VGradient))
  n5 = Multiply(A=OneMinus(None=@n3.R), B=Multiply(A=@n4, B=@n4))
  n6 = TextureSample(UVs=@n2) {samplerType=SAMPLERTYPE_Normal, texture=/BRCosmetics/Characters/Player/Male/Medium/Bodies/M_MED_CycloneUniverse/Textures/Planet_Normal}
  n7 = SetMaterialAttributes!(
   MaterialAttributes
   BaseColor=TextureSample(UVs=@n2) {texture=/BRCosmetics/Characters/Player/Male/Medium/Bodies/M_MED_CycloneUniverse/Textures/Planet_Diffuse}.RGB
   Roughness=LinearInterpolate(Alpha=@n3.R) {constA=0.35, constB=0.8}
   Specular=LinearInterpolate(Alpha=@n3.R) {constA=0.1, constB=0.6}
   Normal=@n6.RGB
  ) {attributeSetTypes [
    #1
    #2
    #3
    #4
   ]}
  FunctionOutput(None=BlendMaterialAttributes(
    A=FunctionInput() {inputType=FunctionInput_MaterialAttributes}
    B=MaterialFunctionCall!(MASK=@n5, Color Change Sequence, Material=@n7) {materialFunction=/Game/Weapons/WeaponSkins/MaterialFunction/GenericGem}.Result
    Alpha=FunctionInput() {inputName=Alpha, inputType=FunctionInput_Scalar}
   ))
 ]}
