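# Master head material for the male medium Cyclone Universe character skin, a Default Lit skeletal mesh material built on material attributes.
# The shared character PreFX function feeds a chain of panning emissive, rotating planet and universe negative effects, then the PostFX function finishes it.
# Channel masks from one SkinFX2 texture place each effect, and a NeckNegativeIntensity scalar feeds the negative effect.
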Material /BRCosmetics/Characters/Player/Male/Medium/Bodies/M_MED_CycloneUniverse/Materials/M_Cyclone_Universe_Head_Master
graph {
 properties {
  bUseMaterialAttributes=true
  bUsedWithMeshDeformer=true
  bUsedWithSkeletalMesh=true
  pixelDepthOffsetMode=PDOM_Legacy
  usedShadingModels=Default Lit
 }
 nodes [
  n1 = MaterialFunctionCall() {materialFunction=&BaseFN_CharacterFunction_PreFX}
  n2 = TextureSampleParameter2D() {parameterName=SkinFX2, texture=&Black}
  n3 = MaterialFunctionCall!(
   Alpha=ChannelMaskParameter(None=@n2.RGB) {parameterName=EmissieMask, defaultValue=<rgba 0 1 0 0>}.R
   isVehicle
   Gmap=@n1.Gmap
   UVOffset
   Local Normal
   Local Position=@n1.Gmap
   EmissiveColorOverride
   Material=@n1.Gmap
  ) {materialFunction=/Game/Weapons/WeaponSkins/MaterialFunction/PanningEmissive}
  n4 = ChannelMaskParameter(None=@n2.RGB) {maskChannel=Green, parameterName=RotatingPlanet, defaultValue=<rgba 0 1 0 0>}
  MaterialFunctionCall(
   AO=@n1.AO
   Distance Blend=@n1.AO
   Fresnel=@n1.AO
   Gmap=@n1.AO
   Material=MaterialFunctionCall!(
    Alpha=ChannelMaskParameter(None=@n2.RGB) {maskChannel=Blue, parameterName=Universe, defaultValue=<rgba 0 0 1 0>}.R
    Gmap=ScalarParameter() {defaultValue=0.1, parameterName=NeckNegativeIntensity}
    In=MaterialFunctionCall!(In=@n3.Result, Alpha=@n4.R) {materialFunction=&/MF_Cyclone_RotatingPlanet}.Result
   ) {materialFunction=/BRCosmetics/Weapons/WeaponSkins/MaterialFunction/MF_CyborgUniverseNegative}.Result
  ) {materialFunction=&BaseFN_CharacterFunction_PostFX}
 ]
}
