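# Physics asset for the Cyclone Space medium male player body, built around a long dynamic coat.
# Front, side and back coat strips are swing/twist-limited capsule chains, cross-linked row by row with free-angular constraints into a cloth-like lattice.
# Tiny MSTR boxes anchor coat chains to spine_04, directly or via upper-coat boxes; hoodie boxes attach to spine_05.
# Capsules cover pelvis, spine and legs, with only a left foot and no arms or head.
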PhysicsAsset /BRCosmetics/Characters/Player/Male/Medium/Bodies/M_MED_CycloneSpace/Meshes/M_MED_Cyclone_Space_Physics
bodies {
 calf_col_r [{shapeType=Capsule, length=18.5, radius=11, center=<13 2 0>, rotation=<rot 90 0 180>}]
 calf_l [{shapeType=Capsule, length=49, radius=11, center=<-25 -2 0.0003>, rotation=<rot 90 0 0>}]
 calf_r [{shapeType=Capsule, length=22.5, radius=8.5, center=<4.3 2 0>, rotation=<rot 90 0 0>}]
 dyn_coat_00_l [{shapeType=Capsule, length=26, radius=4.25, shapeName=dyn_coat_00_l_capsule, center=<5 0 0>, rotation=<rot 0 0 90>}]
 dyn_coat_00_r [{shapeType=Capsule, length=26, radius=4.25, shapeName=dyn_coat_00_r_capsule, center=<5 0 0>, rotation=<rot 0 0 90>}]
 dyn_coat_01_l [{shapeType=Capsule, length=26, radius=4, shapeName=dyn_coat_01_l_capsule, center=<5 0 0>, rotation=<rot 0 0 90>}]
 dyn_coat_01_l_MSTR [{shapeType=Box, extentX=0.01, extentY=0.01, extentZ=0.01, rotation=<rot 0 -90 0>}]
 dyn_coat_01_r [{shapeType=Capsule, length=26, radius=4, shapeName=dyn_coat_01_r_capsule, center=<6 0 0>, rotation=<rot 0 0 90>}]
 dyn_coat_01_r_MSTR [{shapeType=Box, extentX=0.01, extentY=0.01, extentZ=0.01, rotation=<rot 0 -90 0>}]
 dyn_coat_02_l [{shapeType=Capsule, length=26, radius=4, shapeName=dyn_coat_02_l_capsule, center=<5 0 0>, rotation=<rot 0 0 90>}]
 dyn_coat_02_r [{shapeType=Capsule, length=26, radius=4, shapeName=dyn_coat_02_r_capsule, center=<5 0 0>, rotation=<rot 0 0 90>}]
 dyn_coat_03_l [{shapeType=Capsule, length=26, radius=5, shapeName=dyn_coat_03_l_capsule, center=<5 0 0>, rotation=<rot 0 0 90>}]
 dyn_coat_03_r [{shapeType=Capsule, length=26, radius=5, shapeName=dyn_coat_03_r_capsule, center=<5 0 0>, rotation=<rot 0 0 90>}]
 dyn_coat_04_l [{shapeType=Capsule, length=26, radius=5, shapeName=dyn_coat_04_l_capsule, center=<5 0 0>, rotation=<rot 0 0 90>}]
 dyn_coat_04_r [{shapeType=Capsule, length=26, radius=5, shapeName=dyn_coat_04_r_capsule, center=<5 0 0>, rotation=<rot 0 0 90>}]
 dyn_coat_05_l [{shapeType=Capsule, length=26, radius=5, shapeName=dyn_coat_05_l_capsule, center=<6 0 0>, rotation=<rot 0 0 90>}]
 dyn_coat_05_r [{shapeType=Capsule, length=26, radius=5, shapeName=dyn_coat_05_r_capsule, center=<6 0 0>, rotation=<rot 0 0 90>}]
 dyn_coat_06_l [{shapeType=Capsule, length=26, radius=6.5, shapeName=dyn_coat_06_l_capsule, center=<7.5 0 0>, rotation=<rot 0 -10 90>}]
 dyn_coat_06_r [{shapeType=Capsule, length=26, radius=6.5, shapeName=dyn_coat_06_r_capsule, center=<7.5 0 0>, rotation=<rot 0 10 90>}]
 dyn_coat_bk_00 [{shapeType=Capsule, length=26, radius=4.25, shapeName=dyn_coat_bk_00_capsule, center=<5 0 0>, rotation=<rot 0 0 90>}]
 dyn_coat_bk_00_l [{shapeType=Capsule, length=26, radius=4.25, shapeName=dyn_coat_bk_00_l_capsule, center=<5 0 0>, rotation=<rot 0 0 90>}]
 dyn_coat_bk_00_r [{shapeType=Capsule, length=26, radius=4.25, shapeName=dyn_coat_bk_00_r_capsule, center=<5 0 0>, rotation=<rot 0 0 90>}]
 dyn_coat_bk_01 [{shapeType=Capsule, length=26, radius=4, shapeName=dyn_coat_bk_01_capsule, center=<5 0 0>, rotation=<rot 0 0 90>}]
 dyn_coat_bk_01_MSTR [{shapeType=Box, extentX=0.01, extentY=0.01, extentZ=0.01, rotation=<rot 0 -90 0>}]
 dyn_coat_bk_01_l [{shapeType=Capsule, length=26, radius=4, shapeName=dyn_coat_bk_01_l_capsule, center=<5 0 0>, rotation=<rot 0 0 90>}]
 dyn_coat_bk_01_l_MSTR [{shapeType=Box, extentX=0.01, extentY=0.01, extentZ=0.01, rotation=<rot 0 -90 0>}]
 dyn_coat_bk_01_r [{shapeType=Capsule, length=26, radius=4, shapeName=dyn_coat_bk_01_r_capsule, center=<5 0 0>, rotation=<rot 0 0 90>}]
 dyn_coat_bk_01_r_MSTR [{shapeType=Box, extentX=0.01, extentY=0.01, extentZ=0.01, rotation=<rot 0 -90 0>}]
 dyn_coat_bk_02 [{shapeType=Capsule, length=26, radius=4, shapeName=dyn_coat_bk_02_capsule, center=<5 0 0>, rotation=<rot 0 0 90>}]
 dyn_coat_bk_02_l [{shapeType=Capsule, length=26, radius=4, shapeName=dyn_coat_bk_02_l_capsule, center=<5 0 0>, rotation=<rot 0 0 95>}]
 dyn_coat_bk_02_r [{shapeType=Capsule, length=26, radius=4, shapeName=dyn_coat_bk_02_r_capsule, center=<5 0 0>, rotation=<rot 0 0 65>}]
 dyn_coat_bk_03 [{shapeType=Capsule, length=26, radius=5, shapeName=dyn_coat_bk_03_capsule, center=<5 0 0>, rotation=<rot 0 0 90>}]
 dyn_coat_bk_03_l [{shapeType=Capsule, length=26, radius=5, shapeName=dyn_coat_bk_03_l_capsule, center=<5 0 0>, rotation=<rot 0 0 95>}]
 dyn_coat_bk_03_r [{shapeType=Capsule, length=26, radius=5, shapeName=dyn_coat_bk_03_r_capsule, center=<5 0 0>, rotation=<rot 0 0 65>}]
 dyn_coat_bk_04 [{shapeType=Capsule, length=26, radius=5, shapeName=dyn_coat_bk_04_capsule, center=<5 0 0>, rotation=<rot 0 0 90>}]
 dyn_coat_bk_04_l [{shapeType=Capsule, length=26, radius=5, shapeName=dyn_coat_bk_04_l_capsule, center=<5 0 0>, rotation=<rot 0 0 95>}]
 dyn_coat_bk_04_r [{shapeType=Capsule, length=26, radius=5, shapeName=dyn_coat_bk_04_r_capsule, center=<5 0 0>, rotation=<rot 0 0 65>}]
 dyn_coat_bk_05 [{shapeType=Capsule, length=26, radius=5, shapeName=dyn_coat_bk_05_capsule, center=<6 0 0>, rotation=<rot 0 0 90>}]
 dyn_coat_bk_05_l [{shapeType=Capsule, length=26, radius=5, shapeName=dyn_coat_bk_05_l_capsule, center=<6 0 0>, rotation=<rot 0 0 95>}]
 dyn_coat_bk_05_r [{shapeType=Capsule, length=26, radius=5, shapeName=dyn_coat_bk_05_r_capsule, center=<6 0 0>, rotation=<rot 0 0 65>}]
 dyn_coat_bk_06 [{shapeType=Capsule, length=26, radius=6.5, shapeName=dyn_coat_bk_06_capsule, center=<6.3 0 0>, rotation=<rot 0 0 90>}]
 dyn_coat_bk_06_l [{shapeType=Capsule, length=26, radius=6.5, shapeName=dyn_coat_bk_06_l_capsule, center=<6.3 0 0>, rotation=<rot 0 0 95>}]
 dyn_coat_bk_06_r [{shapeType=Capsule, length=26, radius=6.5, shapeName=dyn_coat_bk_06_r_capsule, center=<6.3 0 0>, rotation=<rot 0 0 65>}]
 dyn_coat_fr_00 [{shapeType=Capsule, length=26, radius=4.25, shapeName=dyn_coat_fr_00_capsule, center=<6 0 0>, rotation=<rot 0 -10 90>}]
 dyn_coat_fr_00_r [{shapeType=Capsule, length=26, radius=4.25, shapeName=dyn_coat_fr_00_r_capsule, center=<6 0 0>, rotation=<rot 0 0 90>}]
 dyn_coat_fr_01_l [{shapeType=Capsule, length=26, radius=4, shapeName=dyn_coat_fr_01_l_capsule, center=<6 0 0>, rotation=<rot 0 -10 90>}]
 dyn_coat_fr_01_r [{shapeType=Capsule, length=26, radius=4, shapeName=dyn_coat_fr_01_r_capsule, center=<6 0 0>, rotation=<rot 0 10 90>}]
 dyn_coat_fr_02_l [{shapeType=Capsule, length=26, radius=4.75, shapeName=dyn_coat_fr_02_l_capsule, center=<6.5 0 0>, rotation=<rot 0 -10 90>}]
 dyn_coat_fr_02_r [{shapeType=Capsule, length=26, radius=4.75, shapeName=dyn_coat_fr_02_r_capsule, center=<6.5 0 0>, rotation=<rot 0 10 90>}]
 dyn_coat_fr_03_l [{shapeType=Capsule, length=26, radius=5, shapeName=dyn_coat_fr_03_l_capsule, center=<7 0 0>, rotation=<rot 0 -10 90>}]
 dyn_coat_fr_03_r [{shapeType=Capsule, length=26, radius=5, shapeName=dyn_coat_fr_03_r_capsule, center=<7 0 0>, rotation=<rot 0 10 90>}]
 dyn_coat_fr_04_l [{shapeType=Capsule, length=26, radius=5, shapeName=dyn_coat_fr_04_l_capsule, center=<6.5 0 0>, rotation=<rot 0 -10 90>}]
 dyn_coat_fr_04_r [{shapeType=Capsule, length=26, radius=5, shapeName=dyn_coat_fr_04_r_capsule, center=<6.5 0 0>, rotation=<rot 0 10 90>}]
 dyn_coat_fr_05_l [{shapeType=Capsule, length=26, radius=5, shapeName=dyn_coat_fr_05_l_capsule, center=<8 0 0>, rotation=<rot 0 -10 90>}]
 dyn_coat_fr_05_r [{shapeType=Capsule, length=26, radius=5, shapeName=dyn_coat_fr_05_r_capsule, center=<8 0 0>, rotation=<rot 0 10 90>}]
 dyn_coat_fr_06_l [{shapeType=Capsule, length=26, radius=5, shapeName=dyn_coat_fr_06_l_capsule, center=<10.5 0 0>, rotation=<rot 0 -10 90>}]
 dyn_coat_fr_06_r [{shapeType=Capsule, length=26, radius=5, shapeName=dyn_coat_fr_06_r_capsule, center=<10.5 0 0>, rotation=<rot 0 10 90>}]
 dyn_coat_fr_0_l [{shapeType=Capsule, length=26, radius=4, shapeName=dyn_coat_fr_0_l_capsule, center=<6 0 0>, rotation=<rot 0 0 90>}]
 dyn_coat_fr_0_l_MSTR [{shapeType=Box, extentX=0.01, extentY=0.01, extentZ=0.01}]
 dyn_coat_fr_0_r [{shapeType=Capsule, length=26, radius=4, shapeName=dyn_coat_fr_0_r_capsule, center=<6 0 0>, rotation=<rot 0 0 90>}]
 dyn_coat_fr_0_r_MSTR [{shapeType=Box, extentX=0.01, extentY=0.01, extentZ=0.01}]
 dyn_coatupper_md_01_l_MSTR [{shapeType=Box, extentX=0.01, extentY=0.01, extentZ=0.01, rotation=<rot 0 -90 0>}]
 dyn_coatupper_md_01_r_MSTR [{shapeType=Box, extentX=0.01, extentY=0.01, extentZ=0.01, rotation=<rot 0 -90 0>}]
 dyn_coatupper_tp_01_l_MSTR [{shapeType=Box, extentX=0.01, extentY=0.01, extentZ=0.01, rotation=<rot 0 -90 0>}]
 dyn_coatupper_tp_01_r_MSTR [{shapeType=Box, extentX=0.01, extentY=0.01, extentZ=0.01, rotation=<rot 0 -90 0>}]
 dyn_hoodie_01_MSTR [{shapeType=Box, extentX=0.01, extentY=0.01, extentZ=0.01, rotation=<rot 0 -90 0>}]
 dyn_hoodie_01_l_MSTR [{shapeType=Box, extentX=0.01, extentY=0.01, extentZ=0.01, rotation=<rot 0 -90 0>}]
 dyn_hoodie_01_r_MSTR [{shapeType=Box, extentX=0.01, extentY=0.01, extentZ=0.01, rotation=<rot 0 -90 0>}]
 dyn_hoodie_bk_01_MSTR [{shapeType=Box, extentX=0.01, extentY=0.01, extentZ=0.01, rotation=<rot 0 -90 0>}]
 dyn_hoodie_bk_01_l_MSTR [{shapeType=Box, extentX=0.01, extentY=0.01, extentZ=0.01, rotation=<rot 0 -90 0>}]
 dyn_hoodie_bk_01_r_MSTR [{shapeType=Box, extentX=0.01, extentY=0.01, extentZ=0.01, rotation=<rot 0 -90 0>}]
 dyn_hoodie_fr_01_l_MSTR [{shapeType=Box, extentX=0.01, extentY=0.01, extentZ=0.01, rotation=<rot 0 -90 0>}]
 dyn_hoodie_fr_01_r_MSTR [{shapeType=Box, extentX=0.01, extentY=0.01, extentZ=0.01, rotation=<rot 0 -90 0>}]
 dyn_pelviscol_01 [{shapeType=Capsule, length=42.4123, radius=3.518, center=<20.4486 -0.0001 0>, rotation=<rot 90 0 180>}]
 foot_l [{shapeType=Capsule, length=19, radius=10, center=<-6 5.25 0>, rotation=<rot 0 0 90>}]
 pelvis [{shapeType=Capsule, length=11.3041, radius=8.6558, center=<-0.1224 3.5186 -0.0053>, rotation=<rot -0.0373 -90 0.2129>}]
 spine_01 [{shapeType=Capsule, length=11.3041, radius=8.6558, center=<-0.1224 3.5186 -0.0053>, rotation=<rot -0.0373 -90 0.2129>}]
 spine_02 [{shapeType=Capsule, length=11.3041, radius=8.6558, center=<-0.1224 3.5186 -0.0053>, rotation=<rot -0.0373 -90 0.2129>}]
 spine_03 [{shapeType=Capsule, length=11.3041, radius=8.6558, center=<-0.1224 3.5186 -0.0053>, rotation=<rot -0.0373 -90 0.2129>}]
 spine_04 [{shapeType=Capsule, length=11.3041, radius=8.6558, center=<-0.1224 3.5186 -0.0053>, rotation=<rot -0.0373 -90 0.2129>}]
 spine_05 [{shapeType=Capsule, length=22.3041, radius=8.6558, center=<-0.1224 3.5186 -0.0053>, rotation=<rot -0.0373 -90 0.2129>}]
 thigh_l [{shapeType=Capsule, length=37.6623, radius=8.268, center=<-20.4486 0.0001 0>, rotation=<rot 90 0 0>}]
 thigh_r [{shapeType=Capsule, length=36.4123, radius=9.518, center=<20.4486 -0.0001 0>, rotation=<rot 90 0 180>}]
}
constraints [
 {bone1Name=dyn_coat_01_l, bone2Name=dyn_coat_00_l, swing1LimitDegrees=45, swing2LimitDegrees=45, twistLimitDegrees=5}
 {bone1Name=dyn_coat_02_l, bone2Name=dyn_coat_01_l, swing1LimitDegrees=45, swing2LimitDegrees=45, twistLimitDegrees=5}
 {bone1Name=dyn_coat_03_l, bone2Name=dyn_coat_02_l, swing1LimitDegrees=45, swing2LimitDegrees=45, twistLimitDegrees=5}
 {bone1Name=dyn_coat_04_l, bone2Name=dyn_coat_03_l, swing1LimitDegrees=45, swing2LimitDegrees=45, twistLimitDegrees=5}
 {bone1Name=dyn_coat_05_l, bone2Name=dyn_coat_04_l, swing1LimitDegrees=45, swing2LimitDegrees=45, twistLimitDegrees=5}
 {bone1Name=dyn_coat_06_l, bone2Name=dyn_coat_05_l, swing1LimitDegrees=45, swing2LimitDegrees=45, twistLimitDegrees=5}
 {bone1Name=dyn_coat_bk_01, bone2Name=dyn_coat_bk_00, swing1LimitDegrees=30, swing2LimitDegrees=30, twistLimitDegrees=5}
 {bone1Name=dyn_coat_bk_02, bone2Name=dyn_coat_bk_01, swing1LimitDegrees=30, swing2LimitDegrees=30, twistLimitDegrees=5}
 {bone1Name=dyn_coat_bk_03, bone2Name=dyn_coat_bk_02, swing1LimitDegrees=30, swing2LimitDegrees=30, twistLimitDegrees=5}
 {bone1Name=dyn_coat_bk_04, bone2Name=dyn_coat_bk_03, swing1LimitDegrees=30, swing2LimitDegrees=30, twistLimitDegrees=5}
 {bone1Name=dyn_coat_bk_05, bone2Name=dyn_coat_bk_04, swing1LimitDegrees=30, swing2LimitDegrees=30, twistLimitDegrees=5}
 {bone1Name=dyn_coat_bk_06, bone2Name=dyn_coat_bk_05, swing1LimitDegrees=30, swing2LimitDegrees=30, twistLimitDegrees=5}
 {bone1Name=dyn_coat_bk_01_l, bone2Name=dyn_coat_bk_00_l, swing1LimitDegrees=45, swing2LimitDegrees=45, twistLimitDegrees=5}
 {bone1Name=dyn_coat_bk_02_l, bone2Name=dyn_coat_bk_01_l, swing1LimitDegrees=45, swing2LimitDegrees=45, twistLimitDegrees=5}
 {bone1Name=dyn_coat_bk_03_l, bone2Name=dyn_coat_bk_02_l, swing1LimitDegrees=45, swing2LimitDegrees=45, twistLimitDegrees=5}
 {bone1Name=dyn_coat_bk_04_l, bone2Name=dyn_coat_bk_03_l, swing1LimitDegrees=45, swing2LimitDegrees=45, twistLimitDegrees=5}
 {bone1Name=dyn_coat_bk_05_l, bone2Name=dyn_coat_bk_04_l, swing1LimitDegrees=45, swing2LimitDegrees=45, twistLimitDegrees=5}
 {bone1Name=dyn_coat_bk_06_l, bone2Name=dyn_coat_bk_05_l, swing1LimitDegrees=45, swing2LimitDegrees=45, twistLimitDegrees=5}
 {bone1Name=dyn_coat_fr_01_l, bone2Name=dyn_coat_fr_00, swing1LimitDegrees=45, swing2LimitDegrees=45, twistLimitDegrees=5}
 {bone1Name=dyn_coat_fr_02_l, bone2Name=dyn_coat_fr_01_l, swing1LimitDegrees=45, swing2LimitDegrees=45, twistLimitDegrees=5}
 {bone1Name=dyn_coat_fr_03_l, bone2Name=dyn_coat_fr_02_l, swing1LimitDegrees=45, swing2LimitDegrees=45, twistLimitDegrees=5}
 {bone1Name=dyn_coat_fr_04_l, bone2Name=dyn_coat_fr_03_l, swing1LimitDegrees=45, swing2LimitDegrees=45, twistLimitDegrees=5}
 {bone1Name=dyn_coat_fr_05_l, bone2Name=dyn_coat_fr_04_l, swing1LimitDegrees=45, swing2LimitDegrees=45, twistLimitDegrees=5}
 {bone1Name=dyn_coat_fr_06_l, bone2Name=dyn_coat_fr_05_l, swing1LimitDegrees=45, swing2LimitDegrees=45, twistLimitDegrees=5}
 {bone1Name=dyn_coat_fr_01_r, bone2Name=dyn_coat_fr_00_r, swing1LimitDegrees=45, swing2LimitDegrees=45, twistLimitDegrees=5}
 {bone1Name=dyn_coat_fr_02_r, bone2Name=dyn_coat_fr_01_r, swing1LimitDegrees=45, swing2LimitDegrees=45, twistLimitDegrees=5}
 {bone1Name=dyn_coat_fr_03_r, bone2Name=dyn_coat_fr_02_r, swing1LimitDegrees=45, swing2LimitDegrees=45, twistLimitDegrees=5}
 {bone1Name=dyn_coat_fr_04_r, bone2Name=dyn_coat_fr_03_r, swing1LimitDegrees=45, swing2LimitDegrees=45, twistLimitDegrees=5}
 {bone1Name=dyn_coat_fr_05_r, bone2Name=dyn_coat_fr_04_r, swing1LimitDegrees=45, swing2LimitDegrees=45, twistLimitDegrees=5}
 {bone1Name=dyn_coat_fr_06_r, bone2Name=dyn_coat_fr_05_r, swing1LimitDegrees=45, swing2LimitDegrees=45, twistLimitDegrees=5}
 {bone1Name=dyn_coat_01_r, bone2Name=dyn_coat_00_r, swing1LimitDegrees=45, swing2LimitDegrees=45, twistLimitDegrees=5}
 {bone1Name=dyn_coat_02_r, bone2Name=dyn_coat_01_r, swing1LimitDegrees=45, swing2LimitDegrees=45, twistLimitDegrees=5}
 {bone1Name=dyn_coat_03_r, bone2Name=dyn_coat_02_r, swing1LimitDegrees=45, swing2LimitDegrees=45, twistLimitDegrees=5}
 {bone1Name=dyn_coat_04_r, bone2Name=dyn_coat_03_r, swing1LimitDegrees=45, swing2LimitDegrees=45, twistLimitDegrees=5}
 {bone1Name=dyn_coat_05_r, bone2Name=dyn_coat_04_r, swing1LimitDegrees=45, swing2LimitDegrees=45, twistLimitDegrees=5}
 {bone1Name=dyn_coat_06_r, bone2Name=dyn_coat_05_r, swing1LimitDegrees=45, swing2LimitDegrees=45, twistLimitDegrees=5}
 {bone1Name=dyn_coat_bk_01_r, bone2Name=dyn_coat_bk_00_r, swing1LimitDegrees=30, swing2LimitDegrees=30, twistLimitDegrees=5}
 {bone1Name=dyn_coat_bk_02_r, bone2Name=dyn_coat_bk_01_r, swing1LimitDegrees=30, swing2LimitDegrees=30, twistLimitDegrees=5}
 {bone1Name=dyn_coat_bk_03_r, bone2Name=dyn_coat_bk_02_r, swing1LimitDegrees=30, swing2LimitDegrees=30, twistLimitDegrees=5}
 {bone1Name=dyn_coat_bk_04_r, bone2Name=dyn_coat_bk_03_r, swing1LimitDegrees=30, swing2LimitDegrees=30, twistLimitDegrees=5}
 {bone1Name=dyn_coat_bk_05_r, bone2Name=dyn_coat_bk_04_r, swing1LimitDegrees=30, swing2LimitDegrees=30, twistLimitDegrees=5}
 {bone1Name=dyn_coat_bk_06_r, bone2Name=dyn_coat_bk_05_r, swing1LimitDegrees=30, swing2LimitDegrees=30, twistLimitDegrees=5}
 {
  bone1Name=calf_l
  bone2Name=thigh_l
  swing1LimitDegrees=45
  swing1Motion=Free
  swing2LimitDegrees=45
  swing2Motion=Free
  twistLimitDegrees=45
  twistMotion=Free
 }
 {
  bone1Name=dyn_coat_06_l
  bone2Name=dyn_coat_fr_06_l
  swing1LimitDegrees=45
  swing1Motion=Free
  swing2LimitDegrees=45
  swing2Motion=Free
  twistLimitDegrees=45
  twistMotion=Free
 }
 {
  bone1Name=dyn_coat_05_l
  bone2Name=dyn_coat_fr_05_l
  swing1LimitDegrees=45
  swing1Motion=Free
  swing2LimitDegrees=45
  swing2Motion=Free
  twistLimitDegrees=45
  twistMotion=Free
 }
 {
  bone1Name=dyn_coat_04_l
  bone2Name=dyn_coat_fr_04_l
  swing1LimitDegrees=45
  swing1Motion=Free
  swing2LimitDegrees=45
  swing2Motion=Free
  twistLimitDegrees=45
  twistMotion=Free
 }
 {
  bone1Name=dyn_coat_03_l
  bone2Name=dyn_coat_fr_03_l
  swing1LimitDegrees=45
  swing1Motion=Free
  swing2LimitDegrees=45
  swing2Motion=Free
  twistLimitDegrees=45
  twistMotion=Free
 }
 {
  bone1Name=dyn_coat_02_l
  bone2Name=dyn_coat_fr_02_l
  swing1LimitDegrees=45
  swing1Motion=Free
  swing2LimitDegrees=45
  swing2Motion=Free
  twistLimitDegrees=45
  twistMotion=Free
 }
 {
  bone1Name=dyn_coat_01_l
  bone2Name=dyn_coat_fr_01_l
  swing1LimitDegrees=45
  swing1Motion=Free
  swing2LimitDegrees=45
  swing2Motion=Free
  twistLimitDegrees=45
  twistMotion=Free
 }
 {
  bone1Name=dyn_coat_01_l
  bone2Name=dyn_coat_bk_01_l
  swing1LimitDegrees=45
  swing1Motion=Free
  swing2LimitDegrees=45
  swing2Motion=Free
  twistLimitDegrees=45
  twistMotion=Free
 }
 {
  bone1Name=dyn_coat_02_l
  bone2Name=dyn_coat_bk_02_l
  swing1LimitDegrees=45
  swing1Motion=Free
  swing2LimitDegrees=45
  swing2Motion=Free
  twistLimitDegrees=45
  twistMotion=Free
 }
 {
  bone1Name=dyn_coat_03_l
  bone2Name=dyn_coat_bk_03_l
  swing1LimitDegrees=45
  swing1Motion=Free
  swing2LimitDegrees=45
  swing2Motion=Free
  twistLimitDegrees=45
  twistMotion=Free
 }
 {
  bone1Name=dyn_coat_04_l
  bone2Name=dyn_coat_bk_04_l
  swing1LimitDegrees=45
  swing1Motion=Free
  swing2LimitDegrees=45
  swing2Motion=Free
  twistLimitDegrees=45
  twistMotion=Free
 }
 {
  bone1Name=dyn_coat_05_l
  bone2Name=dyn_coat_bk_05_l
  swing1LimitDegrees=45
  swing1Motion=Free
  swing2LimitDegrees=45
  swing2Motion=Free
  twistLimitDegrees=45
  twistMotion=Free
 }
 {
  bone1Name=dyn_coat_06_l
  bone2Name=dyn_coat_bk_06_l
  swing1LimitDegrees=45
  swing1Motion=Free
  swing2LimitDegrees=45
  swing2Motion=Free
  twistLimitDegrees=45
  twistMotion=Free
 }
 {
  bone1Name=dyn_coat_bk_01
  bone2Name=dyn_coat_bk_01_l
  swing1LimitDegrees=45
  swing1Motion=Free
  swing2LimitDegrees=45
  swing2Motion=Free
  twistLimitDegrees=45
  twistMotion=Free
 }
 {
  bone1Name=dyn_coat_bk_02
  bone2Name=dyn_coat_bk_02_l
  swing1LimitDegrees=45
  swing1Motion=Free
  swing2LimitDegrees=45
  swing2Motion=Free
  twistLimitDegrees=45
  twistMotion=Free
 }
 {
  bone1Name=dyn_coat_bk_03
  bone2Name=dyn_coat_bk_03_l
  swing1LimitDegrees=45
  swing1Motion=Free
  swing2LimitDegrees=45
  swing2Motion=Free
  twistLimitDegrees=45
  twistMotion=Free
 }
 {
  bone1Name=dyn_coat_bk_04
  bone2Name=dyn_coat_bk_04_l
  swing1LimitDegrees=45
  swing1Motion=Free
  swing2LimitDegrees=45
  swing2Motion=Free
  twistLimitDegrees=45
  twistMotion=Free
 }
 {
  bone1Name=dyn_coat_bk_05
  bone2Name=dyn_coat_bk_05_l
  swing1LimitDegrees=45
  swing1Motion=Free
  swing2LimitDegrees=45
  swing2Motion=Free
  twistLimitDegrees=45
  twistMotion=Free
 }
 {
  bone1Name=dyn_coat_bk_06
  bone2Name=dyn_coat_bk_06_l
  swing1LimitDegrees=45
  swing1Motion=Free
  swing2LimitDegrees=45
  swing2Motion=Free
  twistLimitDegrees=45
  twistMotion=Free
 }
 {
  bone1Name=dyn_coat_bk_01
  bone2Name=dyn_coat_bk_01_r
  swing1LimitDegrees=45
  swing1Motion=Free
  swing2LimitDegrees=45
  swing2Motion=Free
  twistLimitDegrees=45
  twistMotion=Free
 }
 {
  bone1Name=dyn_coat_bk_02
  bone2Name=dyn_coat_bk_02_r
  swing1LimitDegrees=45
  swing1Motion=Free
  swing2LimitDegrees=45
  swing2Motion=Free
  twistLimitDegrees=45
  twistMotion=Free
 }
 {
  bone1Name=dyn_coat_bk_03
  bone2Name=dyn_coat_bk_03_r
  swing1LimitDegrees=45
  swing1Motion=Free
  swing2LimitDegrees=45
  swing2Motion=Free
  twistLimitDegrees=45
  twistMotion=Free
 }
 {
  bone1Name=dyn_coat_bk_04
  bone2Name=dyn_coat_bk_04_r
  swing1LimitDegrees=45
  swing1Motion=Free
  swing2LimitDegrees=45
  swing2Motion=Free
  twistLimitDegrees=45
  twistMotion=Free
 }
 {
  bone1Name=dyn_coat_bk_05
  bone2Name=dyn_coat_bk_05_r
  swing1LimitDegrees=45
  swing1Motion=Free
  swing2LimitDegrees=45
  swing2Motion=Free
  twistLimitDegrees=45
  twistMotion=Free
 }
 {
  bone1Name=dyn_coat_bk_06
  bone2Name=dyn_coat_bk_06_r
  swing1LimitDegrees=45
  swing1Motion=Free
  swing2LimitDegrees=45
  swing2Motion=Free
  twistLimitDegrees=45
  twistMotion=Free
 }
 {
  bone1Name=dyn_coat_01_r
  bone2Name=dyn_coat_bk_01_r
  swing1LimitDegrees=45
  swing1Motion=Free
  swing2LimitDegrees=45
  swing2Motion=Free
  twistLimitDegrees=45
  twistMotion=Free
 }
 {
  bone1Name=dyn_coat_02_r
  bone2Name=dyn_coat_bk_02_r
  swing1LimitDegrees=45
  swing1Motion=Free
  swing2LimitDegrees=45
  swing2Motion=Free
  twistLimitDegrees=45
  twistMotion=Free
 }
 {
  bone1Name=dyn_coat_03_r
  bone2Name=dyn_coat_bk_03_r
  swing1LimitDegrees=45
  swing1Motion=Free
  swing2LimitDegrees=45
  swing2Motion=Free
  twistLimitDegrees=45
  twistMotion=Free
 }
 {
  bone1Name=dyn_coat_04_r
  bone2Name=dyn_coat_bk_04_r
  swing1LimitDegrees=45
  swing1Motion=Free
  swing2LimitDegrees=45
  swing2Motion=Free
  twistLimitDegrees=45
  twistMotion=Free
 }
 {
  bone1Name=dyn_coat_05_r
  bone2Name=dyn_coat_bk_05_r
  swing1LimitDegrees=45
  swing1Motion=Free
  swing2LimitDegrees=45
  swing2Motion=Free
  twistLimitDegrees=45
  twistMotion=Free
 }
 {
  bone1Name=dyn_coat_06_r
  bone2Name=dyn_coat_bk_06_r
  swing1LimitDegrees=45
  swing1Motion=Free
  swing2LimitDegrees=45
  swing2Motion=Free
  twistLimitDegrees=45
  twistMotion=Free
 }
 {
  bone1Name=dyn_coat_01_r
  bone2Name=dyn_coat_fr_01_r
  swing1LimitDegrees=45
  swing1Motion=Free
  swing2LimitDegrees=45
  swing2Motion=Free
  twistLimitDegrees=45
  twistMotion=Free
 }
 {
  bone1Name=dyn_coat_02_r
  bone2Name=dyn_coat_fr_02_r
  swing1LimitDegrees=45
  swing1Motion=Free
  swing2LimitDegrees=45
  swing2Motion=Free
  twistLimitDegrees=45
  twistMotion=Free
 }
 {
  bone1Name=dyn_coat_03_r
  bone2Name=dyn_coat_fr_03_r
  swing1LimitDegrees=45
  swing1Motion=Free
  swing2LimitDegrees=45
  swing2Motion=Free
  twistLimitDegrees=45
  twistMotion=Free
 }
 {
  bone1Name=dyn_coat_04_r
  bone2Name=dyn_coat_fr_04_r
  swing1LimitDegrees=45
  swing1Motion=Free
  swing2LimitDegrees=45
  swing2Motion=Free
  twistLimitDegrees=45
  twistMotion=Free
 }
 {
  bone1Name=dyn_coat_05_r
  bone2Name=dyn_coat_fr_05_r
  swing1LimitDegrees=45
  swing1Motion=Free
  swing2LimitDegrees=45
  swing2Motion=Free
  twistLimitDegrees=45
  twistMotion=Free
 }
 {
  bone1Name=dyn_coat_06_r
  bone2Name=dyn_coat_fr_06_r
  swing1LimitDegrees=45
  swing1Motion=Free
  swing2LimitDegrees=45
  swing2Motion=Free
  twistLimitDegrees=45
  twistMotion=Free
 }
 {bone1Name=dyn_coat_bk_00, bone2Name=dyn_coat_bk_01_MSTR, swing1LimitDegrees=30, swing2LimitDegrees=30, twistLimitDegrees=5}
 {bone1Name=dyn_coat_bk_00_l, bone2Name=dyn_coat_bk_01_l_MSTR, swing1LimitDegrees=30, swing2LimitDegrees=30, twistLimitDegrees=5}
 {bone1Name=dyn_coat_00_l, bone2Name=dyn_coat_01_l_MSTR, swing1LimitDegrees=30, swing2LimitDegrees=30, twistLimitDegrees=5}
 {bone1Name=dyn_coat_00_r, bone2Name=dyn_coat_01_r_MSTR, swing1LimitDegrees=30, swing2LimitDegrees=30, twistLimitDegrees=5}
 {bone1Name=dyn_coat_bk_00_r, bone2Name=dyn_coat_bk_01_r_MSTR, swing1LimitDegrees=30, swing2LimitDegrees=30, twistLimitDegrees=5}
 {
  bone1Name=foot_l
  bone2Name=calf_l
  swing1LimitDegrees=45
  swing1Motion=Free
  swing2LimitDegrees=45
  swing2Motion=Free
  twistLimitDegrees=45
  twistMotion=Free
 }
 {bone1Name=dyn_coat_fr_00, bone2Name=dyn_coat_fr_0_l, swing1LimitDegrees=45, swing2LimitDegrees=45, twistLimitDegrees=5}
 {bone1Name=dyn_coat_fr_0_l, bone2Name=dyn_coat_fr_0_l_MSTR, swing1LimitDegrees=45, swing2LimitDegrees=45, twistLimitDegrees=5}
 {bone1Name=dyn_coat_fr_00_r, bone2Name=dyn_coat_fr_0_r, swing1LimitDegrees=45, swing2LimitDegrees=45, twistLimitDegrees=5}
 {bone1Name=dyn_coat_fr_0_r, bone2Name=dyn_coat_fr_0_r_MSTR, swing1LimitDegrees=45, swing2LimitDegrees=45, twistLimitDegrees=5}
 {
  bone1Name=dyn_coat_fr_0_r_MSTR
  bone2Name=dyn_coatupper_md_01_r_MSTR
  swing1LimitDegrees=45
  swing1Motion=Free
  swing2LimitDegrees=45
  swing2Motion=Free
  twistLimitDegrees=45
  twistMotion=Free
 }
 {
  bone1Name=dyn_coat_fr_0_l_MSTR
  bone2Name=dyn_coatupper_md_01_l_MSTR
  swing1LimitDegrees=45
  swing1Motion=Free
  swing2LimitDegrees=45
  swing2Motion=Free
  twistLimitDegrees=45
  twistMotion=Free
 }
 {
  bone1Name=calf_col_r
  bone2Name=calf_r
  swing1LimitDegrees=45
  swing1Motion=Free
  swing2LimitDegrees=45
  swing2Motion=Free
  twistLimitDegrees=45
  twistMotion=Free
 }
 {
  bone1Name=dyn_hoodie_01_l_MSTR
  bone2Name=spine_05
  swing1LimitDegrees=45
  swing1Motion=Free
  swing2LimitDegrees=45
  swing2Motion=Free
  twistLimitDegrees=45
  twistMotion=Free
 }
 {
  bone1Name=dyn_hoodie_bk_01_l_MSTR
  bone2Name=spine_05
  swing1LimitDegrees=45
  swing1Motion=Free
  swing2LimitDegrees=45
  swing2Motion=Free
  twistLimitDegrees=45
  twistMotion=Free
 }
 {
  bone1Name=dyn_hoodie_bk_01_MSTR
  bone2Name=spine_05
  swing1LimitDegrees=45
  swing1Motion=Free
  swing2LimitDegrees=45
  swing2Motion=Free
  twistLimitDegrees=45
  twistMotion=Free
 }
 {
  bone1Name=dyn_hoodie_01_r_MSTR
  bone2Name=spine_05
  swing1LimitDegrees=45
  swing1Motion=Free
  swing2LimitDegrees=45
  swing2Motion=Free
  twistLimitDegrees=45
  twistMotion=Free
 }
 {
  bone1Name=dyn_hoodie_bk_01_r_MSTR
  bone2Name=spine_05
  swing1LimitDegrees=45
  swing1Motion=Free
  swing2LimitDegrees=45
  swing2Motion=Free
  twistLimitDegrees=45
  twistMotion=Free
 }
 {
  bone1Name=dyn_hoodie_01_MSTR
  bone2Name=spine_05
  swing1LimitDegrees=45
  swing1Motion=Free
  swing2LimitDegrees=45
  swing2Motion=Free
  twistLimitDegrees=45
  twistMotion=Free
 }
 {
  bone1Name=dyn_hoodie_fr_01_l_MSTR
  bone2Name=spine_05
  swing1LimitDegrees=45
  swing1Motion=Free
  swing2LimitDegrees=45
  swing2Motion=Free
  twistLimitDegrees=45
  twistMotion=Free
 }
 {
  bone1Name=dyn_hoodie_fr_01_r_MSTR
  bone2Name=spine_05
  swing1LimitDegrees=45
  swing1Motion=Free
  swing2LimitDegrees=45
  swing2Motion=Free
  twistLimitDegrees=45
  twistMotion=Free
 }
 {
  bone1Name=dyn_coatupper_tp_01_l_MSTR
  bone2Name=spine_04
  swing1LimitDegrees=45
  swing1Motion=Free
  swing2LimitDegrees=45
  swing2Motion=Free
  twistLimitDegrees=45
  twistMotion=Free
 }
 {
  bone1Name=dyn_coatupper_tp_01_r_MSTR
  bone2Name=spine_04
  swing1LimitDegrees=45
  swing1Motion=Free
  swing2LimitDegrees=45
  swing2Motion=Free
  twistLimitDegrees=45
  twistMotion=Free
 }
 {
  bone1Name=dyn_coatupper_md_01_r_MSTR
  bone2Name=spine_04
  swing1LimitDegrees=45
  swing1Motion=Free
  swing2LimitDegrees=45
  swing2Motion=Free
  twistLimitDegrees=45
  twistMotion=Free
 }
 {
  bone1Name=dyn_coatupper_md_01_l_MSTR
  bone2Name=spine_04
  swing1LimitDegrees=45
  swing1Motion=Free
  swing2LimitDegrees=45
  swing2Motion=Free
  twistLimitDegrees=45
  twistMotion=Free
 }
 {
  bone1Name=dyn_coat_bk_01_MSTR
  bone2Name=spine_04
  swing1LimitDegrees=45
  swing1Motion=Free
  swing2LimitDegrees=45
  swing2Motion=Free
  twistLimitDegrees=45
  twistMotion=Free
 }
 {
  bone1Name=dyn_coat_bk_01_l_MSTR
  bone2Name=spine_04
  swing1LimitDegrees=45
  swing1Motion=Free
  swing2LimitDegrees=45
  swing2Motion=Free
  twistLimitDegrees=45
  twistMotion=Free
 }
 {
  bone1Name=dyn_coat_01_l_MSTR
  bone2Name=spine_04
  swing1LimitDegrees=45
  swing1Motion=Free
  swing2LimitDegrees=45
  swing2Motion=Free
  twistLimitDegrees=45
  twistMotion=Free
 }
 {
  bone1Name=dyn_coat_01_r_MSTR
  bone2Name=spine_04
  swing1LimitDegrees=45
  swing1Motion=Free
  swing2LimitDegrees=45
  swing2Motion=Free
  twistLimitDegrees=45
  twistMotion=Free
 }
 {
  bone1Name=dyn_coat_bk_01_r_MSTR
  bone2Name=spine_04
  swing1LimitDegrees=45
  swing1Motion=Free
  swing2LimitDegrees=45
  swing2Motion=Free
  twistLimitDegrees=45
  twistMotion=Free
 }
 {
  bone1Name=thigh_l
  bone2Name=pelvis
  swing1LimitDegrees=45
  swing1Motion=Free
  swing2LimitDegrees=45
  swing2Motion=Free
  twistLimitDegrees=45
  twistMotion=Free
 }
 {
  bone1Name=thigh_r
  bone2Name=pelvis
  swing1LimitDegrees=45
  swing1Motion=Free
  swing2LimitDegrees=45
  swing2Motion=Free
  twistLimitDegrees=45
  twistMotion=Free
 }
 {
  bone1Name=dyn_pelviscol_01
  bone2Name=pelvis
  swing1LimitDegrees=45
  swing1Motion=Free
  swing2LimitDegrees=45
  swing2Motion=Free
  twistLimitDegrees=45
  twistMotion=Free
 }
 {
  bone1Name=spine_01
  bone2Name=pelvis
  swing1LimitDegrees=45
  swing1Motion=Free
  swing2LimitDegrees=45
  swing2Motion=Free
  twistLimitDegrees=45
  twistMotion=Free
 }
 {
  bone1Name=spine_02
  bone2Name=spine_01
  swing1LimitDegrees=45
  swing1Motion=Free
  swing2LimitDegrees=45
  swing2Motion=Free
  twistLimitDegrees=45
  twistMotion=Free
 }
 {
  bone1Name=spine_03
  bone2Name=spine_02
  swing1LimitDegrees=45
  swing1Motion=Free
  swing2LimitDegrees=45
  swing2Motion=Free
  twistLimitDegrees=45
  twistMotion=Free
 }
 {
  bone1Name=spine_04
  bone2Name=spine_03
  swing1LimitDegrees=45
  swing1Motion=Free
  swing2LimitDegrees=45
  swing2Motion=Free
  twistLimitDegrees=45
  twistMotion=Free
 }
 {
  bone1Name=spine_05
  bone2Name=spine_04
  swing1LimitDegrees=45
  swing1Motion=Free
  swing2LimitDegrees=45
  swing2Motion=Free
  twistLimitDegrees=45
  twistMotion=Free
 }
 {
  bone1Name=calf_r
  bone2Name=thigh_r
  swing1LimitDegrees=45
  swing1Motion=Free
  swing2LimitDegrees=45
  swing2Motion=Free
  twistLimitDegrees=45
  twistMotion=Free
 }
]
