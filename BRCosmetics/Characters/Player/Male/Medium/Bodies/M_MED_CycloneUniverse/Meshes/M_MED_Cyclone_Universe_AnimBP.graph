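# Animation Blueprint for the Cyclone Universe male body, adding secondary motion to coat, hoodie and necklace over the parent mesh's copied pose.
# The main path runs muscle helper constraints and look-ats, then RigidBody simulations whose gravity responds to skydiving and emotes.
# Above LOD 1, or without rigid bodies, thigh constraints and trail controllers take over.
# Fixed additive rotations then shape the coat, with crouch-driven adjustments; the noise-rotator chain is unconnected.
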AnimBlueprint /BRCosmetics/Characters/Player/Male/Medium/Bodies/M_MED_CycloneUniverse/Meshes/M_MED_Cyclone_Universe_AnimBP
target_skeleton=/BRCosmetics/Characters/Player/Male/Medium/Bodies/M_MED_Cyclone/Meshes/M_MED_Cyclone_Skeleton
graphs [
 AnimGraph {nodes [
   n1 = Use cached pose 'trailControlCache'()
   n2 = Use cached pose 'rbanCache'()
   n3 = Get rigids_off() -> (rigids_off:bool)
   n4 = Get rotCoat_0() -> (rotCoat_0:Rotator)
   n5 = "Transform (Modify) Bone - Bone: dyn_coat_bk_00"(
    ComponentPose=Local To Component(LocalPose=Blend Poses by bool(BlendPose_0=@n1, BlendPose_1=@n2, bActiveValue=@n3) {blendType=Linear})
    Alpha=1
    Translation="0,0,0"
    Rotation=@n4
    Scale="1,1,1"
   ) {rotationMode=BMM_Additive, rotationSpace=BCS_BoneSpace, alphaBoolBlend {blendOption=Linear}, boneToModify {boneName=dyn_coat_bk_00}}
   n6 = "Transform (Modify) Bone - Bone: dyn_coat_bk_00_l"(ComponentPose=@n5, Alpha=1, Translation="0,0,0", Rotation=@n4, Scale="1,1,1") {rotationMode=BMM_Additive, rotationSpace=BCS_BoneSpace, alphaBoolBlend {blendOption=Linear}, boneToModify {boneName=dyn_coat_bk_00_l}}
   n7 = "Transform (Modify) Bone - Bone: dyn_coat_00_l"(ComponentPose=@n6, Alpha=1, Translation="0,0,0", Rotation=@n4, Scale="1,1,1") {rotationMode=BMM_Additive, rotationSpace=BCS_BoneSpace, alphaBoolBlend {blendOption=Linear}, boneToModify {boneName=dyn_coat_00_l}}
   n8 = "Transform (Modify) Bone - Bone: dyn_coat_fr_00"(ComponentPose=@n7, Alpha=0, Translation="0,0,0", Rotation=@n4, Scale="1,1,1") {rotationMode=BMM_Additive, rotationSpace=BCS_BoneSpace, alphaBoolBlend {blendOption=Linear}, boneToModify {boneName=dyn_coat_fr_00}}
   n9 = Get rotCoat_1() -> (rotCoat_1:Rotator)
   n10 = "Transform (Modify) Bone - Bone: dyn_coat_bk_00"(ComponentPose=@n8, Alpha=1, Translation="0,0,0", Rotation=@n9, Scale="1,1,1") {rotationMode=BMM_Additive, rotationSpace=BCS_BoneSpace, alphaBoolBlend {blendOption=Linear}, boneToModify {boneName=dyn_coat_bk_00}}
   n11 = Get rotCoat_2() -> (rotCoat_2:Rotator)
   n12 = "Transform (Modify) Bone - Bone: dyn_coat_bk_01"(ComponentPose=@n10, Alpha=1, Translation="0,0,0", Rotation=@n11, Scale="1,1,1") {rotationMode=BMM_Additive, rotationSpace=BCS_BoneSpace, alphaBoolBlend {blendOption=Linear}, boneToModify {boneName=dyn_coat_bk_01}}
   n13 = Get rotCoat_3() -> (rotCoat_3:Rotator)
   n14 = "Transform (Modify) Bone - Bone: dyn_coat_bk_02"(ComponentPose=@n12, Alpha=1, Translation="0,0,0", Rotation=@n13, Scale="1,1,1") {rotationMode=BMM_Additive, rotationSpace=BCS_BoneSpace, alphaBoolBlend {blendOption=Linear}, boneToModify {boneName=dyn_coat_bk_02}}
   n15 = Get rotCoat_4() -> (rotCoat_4:Rotator)
   n16 = "Transform (Modify) Bone - Bone: dyn_coat_bk_03"(ComponentPose=@n14, Alpha=1, Translation="0,0,0", Rotation=@n15, Scale="1,1,1") {rotationMode=BMM_Additive, rotationSpace=BCS_BoneSpace, alphaBoolBlend {blendOption=Linear}, boneToModify {boneName=dyn_coat_bk_03}}
   n17 = Get rotCoat_5() -> (rotCoat_5:Rotator)
   n18 = "Transform (Modify) Bone - Bone: dyn_coat_bk_04"(ComponentPose=@n16, Alpha=1, Translation="0,0,0", Rotation=@n17, Scale="1,1,1") {rotationMode=BMM_Additive, rotationSpace=BCS_BoneSpace, alphaBoolBlend {blendOption=Linear}, boneToModify {boneName=dyn_coat_bk_04}}
   n19 = Get rotCoat_6() -> (rotCoat_6:Rotator)
   n20 = "Transform (Modify) Bone - Bone: dyn_coat_bk_05"(ComponentPose=@n18, Alpha=1, Translation="0,0,0", Rotation=@n19, Scale="1,1,1") {rotationMode=BMM_Additive, rotationSpace=BCS_BoneSpace, alphaBoolBlend {blendOption=Linear}, boneToModify {boneName=dyn_coat_bk_05}}
   n21 = "Transform (Modify) Bone - Bone: dyn_coat_bk_00_l"(ComponentPose=@n20, Alpha=0, Translation="0,0,0", Rotation=@n9, Scale="1,1,1") {rotationMode=BMM_Additive, rotationSpace=BCS_BoneSpace, alphaBoolBlend {blendOption=Linear}, boneToModify {boneName=dyn_coat_bk_00_l}}
   n22 = "Transform (Modify) Bone - Bone: dyn_coat_bk_01_l"(ComponentPose=@n21, Alpha=1, Translation="0,0,0", Rotation=@n11, Scale="1,1,1") {rotationMode=BMM_Additive, rotationSpace=BCS_BoneSpace, alphaBoolBlend {blendOption=Linear}, boneToModify {boneName=dyn_coat_bk_01_l}}
   n23 = "Transform (Modify) Bone - Bone: dyn_coat_bk_02_l"(ComponentPose=@n22, Alpha=1, Translation="0,0,0", Rotation=@n13, Scale="1,1,1") {rotationMode=BMM_Additive, rotationSpace=BCS_BoneSpace, alphaBoolBlend {blendOption=Linear}, boneToModify {boneName=dyn_coat_bk_02_l}}
   n24 = "Transform (Modify) Bone - Bone: dyn_coat_bk_03_l"(ComponentPose=@n23, Alpha=1, Translation="0,0,0", Rotation=@n15, Scale="1,1,1") {rotationMode=BMM_Additive, rotationSpace=BCS_BoneSpace, alphaBoolBlend {blendOption=Linear}, boneToModify {boneName=dyn_coat_bk_03_l}}
   n25 = "Transform (Modify) Bone - Bone: dyn_coat_bk_04_l"(ComponentPose=@n24, Alpha=1, Translation="0,0,0", Rotation=@n17, Scale="1,1,1") {rotationMode=BMM_Additive, rotationSpace=BCS_BoneSpace, alphaBoolBlend {blendOption=Linear}, boneToModify {boneName=dyn_coat_bk_04_l}}
   n26 = "Transform (Modify) Bone - Bone: dyn_coat_bk_05_l"(ComponentPose=@n25, Alpha=1, Translation="0,0,0", Rotation=@n19, Scale="1,1,1") {rotationMode=BMM_Additive, rotationSpace=BCS_BoneSpace, alphaBoolBlend {blendOption=Linear}, boneToModify {boneName=dyn_coat_bk_05_l}}
   n27 = "Transform (Modify) Bone - Bone: dyn_coat_00_l"(ComponentPose=@n26, Alpha=0, Translation="0,0,0", Rotation=@n9, Scale="1,1,1") {rotationMode=BMM_Additive, rotationSpace=BCS_BoneSpace, alphaBoolBlend {blendOption=Linear}, boneToModify {boneName=dyn_coat_00_l}}
   n28 = "Transform (Modify) Bone - Bone: dyn_coat_01_l"(ComponentPose=@n27, Alpha=1, Translation="0,0,0", Rotation=@n11, Scale="1,1,1") {rotationMode=BMM_Additive, rotationSpace=BCS_BoneSpace, alphaBoolBlend {blendOption=Linear}, boneToModify {boneName=dyn_coat_01_l}}
   n29 = "Transform (Modify) Bone - Bone: dyn_coat_02_l"(ComponentPose=@n28, Alpha=1, Translation="0,0,0", Rotation=@n13, Scale="1,1,1") {rotationMode=BMM_Additive, rotationSpace=BCS_BoneSpace, alphaBoolBlend {blendOption=Linear}, boneToModify {boneName=dyn_coat_02_l}}
   n30 = "Transform (Modify) Bone - Bone: dyn_coat_03_l"(ComponentPose=@n29, Alpha=1, Translation="0,0,0", Rotation=@n15, Scale="1,1,1") {rotationMode=BMM_Additive, rotationSpace=BCS_BoneSpace, alphaBoolBlend {blendOption=Linear}, boneToModify {boneName=dyn_coat_03_l}}
   n31 = "Transform (Modify) Bone - Bone: dyn_coat_04_l"(ComponentPose=@n30, Alpha=1, Translation="0,0,0", Rotation=@n17, Scale="1,1,1") {rotationMode=BMM_Additive, rotationSpace=BCS_BoneSpace, alphaBoolBlend {blendOption=Linear}, boneToModify {boneName=dyn_coat_04_l}}
   n32 = "Transform (Modify) Bone - Bone: dyn_coat_05_l"(ComponentPose=@n31, Alpha=1, Translation="0,0,0", Rotation=@n19, Scale="1,1,1") {rotationMode=BMM_Additive, rotationSpace=BCS_BoneSpace, alphaBoolBlend {blendOption=Linear}, boneToModify {boneName=dyn_coat_05_l}}
   n33 = "Transform (Modify) Bone - Bone: dyn_coat_fr_00"(ComponentPose=@n32, Alpha=0, Translation="0,0,0", Rotation=@n9, Scale="1,1,1") {rotationMode=BMM_Additive, rotationSpace=BCS_BoneSpace, alphaBoolBlend {blendOption=Linear}, boneToModify {boneName=dyn_coat_fr_00}}
   n34 = "Transform (Modify) Bone - Bone: dyn_coat_fr_01_l"(ComponentPose=@n33, Alpha=1, Translation="0,0,0", Rotation=@n11, Scale="1,1,1") {rotationMode=BMM_Additive, rotationSpace=BCS_BoneSpace, alphaBoolBlend {blendOption=Linear}, boneToModify {boneName=dyn_coat_fr_01_l}}
   n35 = "Transform (Modify) Bone - Bone: dyn_coat_fr_02_l"(ComponentPose=@n34, Alpha=1, Translation="0,0,0", Rotation=@n13, Scale="1,1,1") {rotationMode=BMM_Additive, rotationSpace=BCS_BoneSpace, alphaBoolBlend {blendOption=Linear}, boneToModify {boneName=dyn_coat_fr_02_l}}
   n36 = "Transform (Modify) Bone - Bone: dyn_coat_fr_03_l"(ComponentPose=@n35, Alpha=1, Translation="0,0,0", Rotation=@n15, Scale="1,1,1") {rotationMode=BMM_Additive, rotationSpace=BCS_BoneSpace, alphaBoolBlend {blendOption=Linear}, boneToModify {boneName=dyn_coat_fr_03_l}}
   n37 = "Transform (Modify) Bone - Bone: dyn_coat_fr_04_l"(ComponentPose=@n36, Alpha=1, Translation="0,0,0", Rotation=@n17, Scale="1,1,1") {rotationMode=BMM_Additive, rotationSpace=BCS_BoneSpace, alphaBoolBlend {blendOption=Linear}, boneToModify {boneName=dyn_coat_fr_04_l}}
   n38 = "Transform (Modify) Bone - Bone: dyn_coat_fr_05_l"(ComponentPose=@n37, Alpha=1, Translation="0,0,0", Rotation=@n19, Scale="1,1,1") {rotationMode=BMM_Additive, rotationSpace=BCS_BoneSpace, alphaBoolBlend {blendOption=Linear}, boneToModify {boneName=dyn_coat_fr_05_l}}
   n39 = Get rotCoat_0()
   n40 = Get isNotCrouchingAlpha() -> (isNotCrouchingAlpha:double)
   n41 = "Transform (Modify) Bone - Bone: dyn_coat_bk_00_r"(ComponentPose=@n38, Alpha=@n40, Translation="0,0,0", Rotation=@n39, Scale="1,1,1") {
    rotationMode=BMM_Additive
    rotationSpace=BCS_BoneSpace
    alphaBoolBlend {blendOption=Linear}
    alphaScaleBiasClamp {bInterpResult=true}
    boneToModify {boneName=dyn_coat_bk_00_r}
   }
   n42 = "Transform (Modify) Bone - Bone: dyn_coat_00_r"(ComponentPose=@n41, Alpha=@n40, Translation="0,0,0", Rotation=@n39, Scale="1,1,1") {
    rotationMode=BMM_Additive
    rotationSpace=BCS_BoneSpace
    alphaBoolBlend {blendOption=Linear}
    alphaScaleBiasClamp {bInterpResult=true}
    boneToModify {boneName=dyn_coat_00_r}
   }
   n43 = "Transform (Modify) Bone - Bone: dyn_coat_fr_00_r"(ComponentPose=@n42, Alpha=0, Translation="0,0,0", Rotation=@n39, Scale="1,1,1") {
    rotationMode=BMM_Additive
    rotationSpace=BCS_BoneSpace
    alphaBoolBlend {blendOption=Linear}
    alphaScaleBiasClamp {bInterpResult=true}
    boneToModify {boneName=dyn_coat_fr_00_r}
   }
   n44 = Get rotCoat_1()
   n45 = "Transform (Modify) Bone - Bone: dyn_coat_bk_00_r"(ComponentPose=@n43, Alpha=@n40, Translation="0,0,0", Rotation=@n44, Scale="1,1,1") {
    rotationMode=BMM_Additive
    rotationSpace=BCS_BoneSpace
    alphaBoolBlend {blendOption=Linear}
    alphaScaleBiasClamp {bInterpResult=true}
    boneToModify {boneName=dyn_coat_bk_00_r}
   }
   n46 = Get rotCoat_2()
   n47 = "Transform (Modify) Bone - Bone: dyn_coat_01_r"(ComponentPose=@n45, Alpha=@n40, Translation="0,0,0", Rotation=@n46, Scale="1,1,1") {
    rotationMode=BMM_Additive
    rotationSpace=BCS_BoneSpace
    alphaBoolBlend {blendOption=Linear}
    alphaScaleBiasClamp {bInterpResult=true}
    boneToModify {boneName=dyn_coat_01_r}
   }
   n48 = Get rotCoat_3()
   n49 = "Transform (Modify) Bone - Bone: dyn_coat_bk_02_r"(ComponentPose=@n47, Alpha=@n40, Translation="0,0,0", Rotation=@n48, Scale="1,1,1") {
    rotationMode=BMM_Additive
    rotationSpace=BCS_BoneSpace
    alphaBoolBlend {blendOption=Linear}
    alphaScaleBiasClamp {bInterpResult=true}
    boneToModify {boneName=dyn_coat_bk_02_r}
   }
   n50 = Get rotCoat_4()
   n51 = "Transform (Modify) Bone - Bone: dyn_coat_bk_03_r"(ComponentPose=@n49, Alpha=@n40, Translation="0,0,0", Rotation=@n50, Scale="1,1,1") {
    rotationMode=BMM_Additive
    rotationSpace=BCS_BoneSpace
    alphaBoolBlend {blendOption=Linear}
    alphaScaleBiasClamp {bInterpResult=true}
    boneToModify {boneName=dyn_coat_bk_03_r}
   }
   n52 = Get rotCoat_5()
   n53 = "Transform (Modify) Bone - Bone: dyn_coat_bk_04_r"(ComponentPose=@n51, Alpha=1, Translation="0,0,0", Rotation=@n52, Scale="1,1,1") {
    rotationMode=BMM_Additive
    rotationSpace=BCS_BoneSpace
    alphaBoolBlend {blendOption=Linear}
    alphaScaleBiasClamp {bInterpResult=true}
    boneToModify {boneName=dyn_coat_bk_04_r}
   }
   n54 = Get rotCoat_6()
   n55 = "Transform (Modify) Bone - Bone: dyn_coat_bk_05_r"(ComponentPose=@n53, Alpha=1, Translation="0,0,0", Rotation=@n54, Scale="1,1,1") {
    rotationMode=BMM_Additive
    rotationSpace=BCS_BoneSpace
    alphaBoolBlend {blendOption=Linear}
    alphaScaleBiasClamp {bInterpResult=true}
    boneToModify {boneName=dyn_coat_bk_05_r}
   }
   n56 = "Transform (Modify) Bone - Bone: dyn_coat_00_r"(ComponentPose=@n55, Alpha=@n40, Translation="0,0,0", Rotation=@n44, Scale="1,1,1") {
    rotationMode=BMM_Additive
    rotationSpace=BCS_BoneSpace
    alphaBoolBlend {blendOption=Linear}
    alphaScaleBiasClamp {bInterpResult=true}
    boneToModify {boneName=dyn_coat_00_r}
   }
   n57 = "Transform (Modify) Bone - Bone: dyn_coat_01_r"(ComponentPose=@n56, Alpha=@n40, Translation="0,0,0", Rotation=@n46, Scale="1,1,1") {
    rotationMode=BMM_Additive
    rotationSpace=BCS_BoneSpace
    alphaBoolBlend {blendOption=Linear}
    alphaScaleBiasClamp {bInterpResult=true}
    boneToModify {boneName=dyn_coat_01_r}
   }
   n58 = "Transform (Modify) Bone - Bone: dyn_coat_02_r"(ComponentPose=@n57, Alpha=@n40, Translation="0,0,0", Rotation=@n48, Scale="1,1,1") {
    rotationMode=BMM_Additive
    rotationSpace=BCS_BoneSpace
    alphaBoolBlend {blendOption=Linear}
    alphaScaleBiasClamp {bInterpResult=true}
    boneToModify {boneName=dyn_coat_02_r}
   }
   n59 = "Transform (Modify) Bone - Bone: dyn_coat_03_r"(ComponentPose=@n58, Alpha=@n40, Translation="0,0,0", Rotation=@n50, Scale="1,1,1") {
    rotationMode=BMM_Additive
    rotationSpace=BCS_BoneSpace
    alphaBoolBlend {blendOption=Linear}
    alphaScaleBiasClamp {bInterpResult=true}
    boneToModify {boneName=dyn_coat_03_r}
   }
   n60 = "Transform (Modify) Bone - Bone: dyn_coat_04_r"(ComponentPose=@n59, Alpha=1, Translation="0,0,0", Rotation=@n52, Scale="1,1,1") {
    rotationMode=BMM_Additive
    rotationSpace=BCS_BoneSpace
    alphaBoolBlend {blendOption=Linear}
    alphaScaleBiasClamp {bInterpResult=true}
    boneToModify {boneName=dyn_coat_04_r}
   }
   n61 = "Transform (Modify) Bone - Bone: dyn_coat_05_r"(ComponentPose=@n60, Alpha=1, Translation="0,0,0", Rotation=@n54, Scale="1,1,1") {
    rotationMode=BMM_Additive
    rotationSpace=BCS_BoneSpace
    alphaBoolBlend {blendOption=Linear}
    alphaScaleBiasClamp {bInterpResult=true}
    boneToModify {boneName=dyn_coat_05_r}
   }
   n62 = "Transform (Modify) Bone - Bone: dyn_coat_fr_00_r"(ComponentPose=@n61, Alpha=@n40, Translation="0,0,0", Rotation=@n44, Scale="1,1,1") {
    rotationMode=BMM_Additive
    rotationSpace=BCS_BoneSpace
    alphaBoolBlend {blendOption=Linear}
    alphaScaleBiasClamp {bInterpResult=true}
    boneToModify {boneName=dyn_coat_fr_00_r}
   }
   n63 = "Transform (Modify) Bone - Bone: dyn_coat_fr_01_r"(ComponentPose=@n62, Alpha=@n40, Translation="0,0,0", Rotation=@n46, Scale="1,1,1") {
    rotationMode=BMM_Additive
    rotationSpace=BCS_BoneSpace
    alphaBoolBlend {blendOption=Linear}
    alphaScaleBiasClamp {bInterpResult=true}
    boneToModify {boneName=dyn_coat_fr_01_r}
   }
   n64 = "Transform (Modify) Bone - Bone: dyn_coat_fr_02_r"(ComponentPose=@n63, Alpha=@n40, Translation="0,0,0", Rotation=@n48, Scale="1,1,1") {
    rotationMode=BMM_Additive
    rotationSpace=BCS_BoneSpace
    alphaBoolBlend {blendOption=Linear}
    alphaScaleBiasClamp {bInterpResult=true}
    boneToModify {boneName=dyn_coat_fr_02_r}
   }
   n65 = "Transform (Modify) Bone - Bone: dyn_coat_fr_03_r"(ComponentPose=@n64, Alpha=@n40, Translation="0,0,0", Rotation=@n50, Scale="1,1,1") {
    rotationMode=BMM_Additive
    rotationSpace=BCS_BoneSpace
    alphaBoolBlend {blendOption=Linear}
    alphaScaleBiasClamp {bInterpResult=true}
    boneToModify {boneName=dyn_coat_fr_03_r}
   }
   n66 = "Transform (Modify) Bone - Bone: dyn_coat_fr_04_r"(ComponentPose=@n65, Alpha=1, Translation="0,0,0", Rotation=@n52, Scale="1,1,1") {
    rotationMode=BMM_Additive
    rotationSpace=BCS_BoneSpace
    alphaBoolBlend {blendOption=Linear}
    alphaScaleBiasClamp {bInterpResult=true}
    boneToModify {boneName=dyn_coat_fr_04_r}
   }
   n67 = "Transform (Modify) Bone - Bone: dyn_coat_fr_05_r"(ComponentPose=@n66, Alpha=1, Translation="0,0,0", Rotation=@n54, Scale="1,1,1") {
    rotationMode=BMM_Additive
    rotationSpace=BCS_BoneSpace
    alphaBoolBlend {blendOption=Linear}
    alphaScaleBiasClamp {bInterpResult=true}
    boneToModify {boneName=dyn_coat_fr_05_r}
   }
   Output Pose(Result=Component To Local(ComponentPose=@n67)) {layerGroup=None}
   n68 = "Constraint - dyn_coat_01_l (P)"(ComponentPose=Local To Component(LocalPose=Use cached pose 'initCache'()), Alpha=1, ConstraintWeights_0=1) {
    alphaBoolBlend {blendOption=Linear}
    boneToModify {boneName=dyn_coat_01_l}
    constraintSetup [{targetBone {boneName=thigh_l}}]
    constraintWeights [1]
   }
   n69 = "Constraint - dyn_coat_fr_01_l (P)"(ComponentPose=@n68, Alpha=1, ConstraintWeights_0=1) {
    alphaBoolBlend {blendOption=Linear}
    boneToModify {boneName=dyn_coat_fr_01_l}
    constraintSetup [{targetBone {boneName=thigh_l}}]
    constraintWeights [1]
   }
   n70 = "Constraint - dyn_coat_bk_01_l (P)"(ComponentPose=@n69, Alpha=0, ConstraintWeights_0=1) {
    alphaBoolBlend {blendOption=Linear}
    boneToModify {boneName=dyn_coat_bk_01_l}
    constraintSetup [{targetBone {boneName=thigh_l}}]
    constraintWeights [1]
   }
   n71 = "Constraint - dyn_coat_01_r (P)"(ComponentPose=@n70, Alpha=1, ConstraintWeights_0=1) {
    alphaBoolBlend {blendOption=Linear}
    boneToModify {boneName=dyn_coat_01_r}
    constraintSetup [{targetBone {boneName=thigh_r}}]
    constraintWeights [1]
   }
   n72 = "Constraint - dyn_coat_fr_01_r (P)"(ComponentPose=@n71, Alpha=1, ConstraintWeights_0=1) {
    alphaBoolBlend {blendOption=Linear}
    boneToModify {boneName=dyn_coat_fr_01_r}
    constraintSetup [{targetBone {boneName=thigh_r}}]
    constraintWeights [1]
   }
   n73 = "Constraint - dyn_coat_bk_01_r (P)"(ComponentPose=@n72, Alpha=0, ConstraintWeights_0=1) {
    alphaBoolBlend {blendOption=Linear}
    boneToModify {boneName=dyn_coat_bk_01_r}
    constraintSetup [{targetBone {boneName=thigh_r}}]
    constraintWeights [1]
   }
   n74 = Get lod2movingAndCrouchAlpha() -> (lod2movingAndCrouchAlpha:double)
   n75 = "Transform (Modify) Bone - Bone: dyn_coat_bk_01"(ComponentPose=@n73, Alpha=@n74, Translation="0,0,0", Rotation="0,0,-35", Scale="1,1,1") {
    rotationMode=BMM_Additive
    alphaBoolBlend {blendOption=Linear}
    alphaScaleBiasClamp {bInterpResult=true}
    boneToModify {boneName=dyn_coat_bk_01}
    rotation {roll=-35}
   }
   n76 = "Transform (Modify) Bone - Bone: dyn_coat_bk_01_r"(ComponentPose=@n75, Alpha=@n74, Translation="0,0,0", Rotation="0,0,-35", Scale="1,1,1") {
    rotationMode=BMM_Additive
    alphaBoolBlend {blendOption=Linear}
    alphaScaleBiasClamp {bInterpResult=true}
    boneToModify {boneName=dyn_coat_bk_01_r}
    rotation {roll=-35}
   }
   n77 = "Transform (Modify) Bone - Bone: dyn_coat_bk_01_l"(ComponentPose=@n76, Alpha=@n74, Translation="0,0,0", Rotation="0,0,-35", Scale="1,1,1") {
    rotationMode=BMM_Additive
    alphaBoolBlend {blendOption=Linear}
    alphaScaleBiasClamp {bInterpResult=true}
    boneToModify {boneName=dyn_coat_bk_01_l}
    rotation {roll=-35}
   }
   n78 = Trail controller - Bone: dyn_coat_bk_06_l!(ComponentPose=@n77, Alpha=1, RelaxationSpeedScale:float=1, FakeVelocity:Vector="0,0,0") -> (Pose) {
    bActorSpaceFakeVel=true
    bLimitRotation=true
    bUsePlanarLimit=true
    chainLength=4
    debugLifeTime=0.1
    maxDeltaTime=0.0333
    relaxationSpeedScale=1.2
    alphaScaleBiasClamp {bInterpResult=true, interpSpeedDecreasing=25, interpSpeedIncreasing=25}
    baseJoint {boneName=root}
    planarLimits [
     {drivingBone {boneName=root}, planeTransform {location=<0 0 10>, rotation=<rot 0 0 0>, scale=<1 1 1>}}
     {drivingBone {boneName=spine_04}, planeTransform {location=<0 40 0>, rotation=<rot 0 0 -90>, scale=<1 1 1>}}
    ]
    rotationLimits [
     {limitMax=<0 0 30>, limitMin=<0 -15 -30>}
     {limitMax=<0 0 30>, limitMin=<0 -15 -30>}
     {limitMax=<0 0 30>, limitMin=<0 -15 -30>}
     {limitMax=<180 180 180>}
    ]
    rotationOffsets [<0 0 0>, <0 0 0>, <0 0 0>, <0 0 0>]
    trailBone {boneName=dyn_coat_bk_06_l}
    trailRelaxationSpeed {editorCurveData {keys [<at 0 15 RCIM_Linear>,<at 0.5289 13.1838 RCIM_Linear>,<at 1 6 RCIM_Linear>]}}
   }
   n79 = Trail controller - Bone: dyn_coat_bk_06!(ComponentPose=@n78, Alpha=1, RelaxationSpeedScale:float=1, FakeVelocity:Vector="0,0,0") -> (Pose) {
    bActorSpaceFakeVel=true
    bLimitRotation=true
    bUsePlanarLimit=true
    chainLength=4
    debugLifeTime=0.1
    maxDeltaTime=0.0333
    relaxationSpeedScale=1.2
    alphaScaleBiasClamp {bInterpResult=true, interpSpeedDecreasing=25, interpSpeedIncreasing=25}
    baseJoint {boneName=root}
    planarLimits [
     {drivingBone {boneName=root}, planeTransform {location=<0 0 10>, rotation=<rot 0 0 0>, scale=<1 1 1>}}
     {drivingBone {boneName=spine_04}, planeTransform {location=<0 40 0>, rotation=<rot 0 0 -90>, scale=<1 1 1>}}
    ]
    rotationLimits [
     {limitMax=<0 0 30>, limitMin=<0 -15 -30>}
     {limitMax=<0 0 30>, limitMin=<0 -15 -30>}
     {limitMax=<0 0 30>, limitMin=<0 -15 -30>}
     {limitMax=<180 180 180>}
    ]
    rotationOffsets [<0 0 0>, <0 0 0>, <0 0 0>, <0 0 0>]
    trailBone {boneName=dyn_coat_bk_06}
    trailRelaxationSpeed {editorCurveData {keys [<at 0 15 RCIM_Linear>,<at 0.5289 13.1838 RCIM_Linear>,<at 1 6 RCIM_Linear>]}}
   }
   n80 = Trail controller - Bone: dyn_coat_bk_06_r!(ComponentPose=@n79, Alpha=1, RelaxationSpeedScale:float=1, FakeVelocity:Vector="0,0,0") -> (Pose) {
    bActorSpaceFakeVel=true
    bLimitRotation=true
    bUsePlanarLimit=true
    chainLength=4
    debugLifeTime=0.1
    maxDeltaTime=0.0333
    relaxationSpeedScale=1.2
    alphaScaleBiasClamp {bInterpResult=true, interpSpeedDecreasing=25, interpSpeedIncreasing=25}
    baseJoint {boneName=root}
    planarLimits [
     {drivingBone {boneName=root}, planeTransform {location=<0 0 10>, rotation=<rot 0 0 0>, scale=<1 1 1>}}
     {drivingBone {boneName=spine_04}, planeTransform {location=<0 40 0>, rotation=<rot 0 0 -90>, scale=<1 1 1>}}
    ]
    rotationLimits [
     {limitMax=<0 0 30>, limitMin=<0 -15 -30>}
     {limitMax=<0 0 30>, limitMin=<0 -15 -30>}
     {limitMax=<0 0 30>, limitMin=<0 -15 -30>}
     {limitMax=<180 180 180>}
    ]
    rotationOffsets [<0 0 0>, <0 0 0>, <0 0 0>, <0 0 0>]
    trailBone {boneName=dyn_coat_bk_06_r}
    trailRelaxationSpeed {editorCurveData {keys [<at 0 15 RCIM_Linear>,<at 0.5289 13.1838 RCIM_Linear>,<at 1 6 RCIM_Linear>]}}
   }
   Save cached pose 'trailControlCache'(Pose=Component To Local(ComponentPose=@n80)) {cacheName=trailControlCache}
   Save cached pose 'initCache'(Pose=Copy Pose From Mesh() {bUseAttachedParent=true}) {cacheName=initCache}
   n81 = "Constraint - scapula_base_l (P)"(ComponentPose=Local To Component(LocalPose=Use cached pose 'initCache'()), Alpha=1, ConstraintWeights_0=1) {
    alphaBoolBlend {blendOption=Linear}
    boneToModify {boneName=scapula_base_l}
    constraintSetup [{targetBone {boneName=clavicle_l}}]
    constraintWeights [1]
   }
   n82 = Look At - Bone: scapula_00_l(ComponentPose=@n81, Alpha=1) {
    alphaBoolBlend {blendOption=Linear}
    boneToModify {boneName=scapula_00_l}
    lookAtLocation {x=0}
    lookAtTarget {boneReference {boneName=scapula_01_l_aim}}
    lookAt_Axis {axis=<1 0>}
    lookUp_Axis {axis {x=0, z=1}}
   }
   n83 = Look At - Bone: pectoralis_01_l(ComponentPose=@n82, Alpha=1) {
    alphaBoolBlend {blendOption=Linear}
    boneToModify {boneName=pectoralis_01_l}
    lookAtLocation {x=0}
    lookAtTarget {boneReference {boneName=upperarm_l}}
    lookAt_Axis {axis=<1 0>}
    lookUp_Axis {axis {x=0, z=1}}
   }
   n84 = Look At - Bone: lattissimus_00_l(ComponentPose=@n83, Alpha=1) {
    alphaBoolBlend {blendOption=Linear}
    boneToModify {boneName=lattissimus_00_l}
    lookAtLocation {x=0}
    lookAtTarget {boneReference {boneName=upperarm_l}}
    lookAt_Axis {axis=<1 0>}
    lookUp_Axis {axis {x=0, z=1}}
   }
   n85 = "Constraint - scapula_base_r (P)"(ComponentPose=@n84, Alpha=1, ConstraintWeights_0=1) {
    alphaBoolBlend {blendOption=Linear}
    boneToModify {boneName=scapula_base_r}
    constraintSetup [{targetBone {boneName=clavicle_r}}]
    constraintWeights [1]
   }
   n86 = Look At - Bone: scapula_00_r(ComponentPose=@n85, Alpha=1) {
    alphaBoolBlend {blendOption=Linear}
    boneToModify {boneName=scapula_00_r}
    lookAtLocation {x=0}
    lookAtTarget {boneReference {boneName=scapula_01_r_aim}}
    lookAt_Axis {axis=<1 0>}
    lookUp_Axis {axis {x=0, z=1}}
   }
   n87 = Look At - Bone: pectoralis_01_r(ComponentPose=@n86, Alpha=1) {
    alphaBoolBlend {blendOption=Linear}
    boneToModify {boneName=pectoralis_01_r}
    lookAtLocation {x=0}
    lookAtTarget {boneReference {boneName=upperarm_r}}
    lookAt_Axis {axis=<1 0>}
    lookUp_Axis {axis {x=0, z=1}}
   }
   n88 = Look At - Bone: lattissimus_00_r(ComponentPose=@n87, Alpha=1) {
    alphaBoolBlend {blendOption=Linear}
    boneToModify {boneName=lattissimus_00_r}
    lookAtLocation {x=0}
    lookAtTarget {boneReference {boneName=upperarm_r}}
    lookAt_Axis {axis=<1 0>}
    lookUp_Axis {axis {x=0, z=1}}
   }
   n89 = "Constraint - dyn_pelviscol_01_aim (PP)"!(ComponentPose=@n88, Alpha=1, ConstraintWeights_0=0.5, ConstraintWeights_1:float=0.5) -> (Pose) {
    alphaBoolBlend {blendOption=Linear}
    boneToModify {boneName=dyn_pelviscol_01_aim}
    constraintSetup [{targetBone {boneName=calf_l}}, {targetBone {boneName=calf_r}}]
    constraintWeights [1,1]
   }
   n90 = Look At - Bone: dyn_pelviscol_01(ComponentPose=@n89, Alpha=1) {
    alphaBoolBlend {blendOption=Linear}
    boneToModify {boneName=dyn_pelviscol_01}
    lookAtTarget {boneReference {boneName=dyn_pelviscol_01_aim}}
    lookAt_Axis {axis=<1 0>}
    lookUp_Axis {axis {x=0, z=1}}
   }
   n91 = Get crouchingAlpha() -> (crouchingAlpha:double)
   n92 = "Constraint - dyn_coat_fr_0_l_MSTR (R)"(ComponentPose=@n90, Alpha=@n91, ConstraintWeights_0=1) {
    alphaBoolBlend {blendOption=Linear}
    alphaScaleBiasClamp {bInterpResult=true}
    boneToModify {boneName=dyn_coat_fr_0_l_MSTR}
    constraintSetup [{transformType=Rotation, targetBone {boneName=thigh_l}}]
    constraintWeights [1]
   }
   n93 = "Constraint - dyn_coat_01_l_MSTR (R)"(ComponentPose=@n92, Alpha=@n91, ConstraintWeights_0=1) {
    alphaBoolBlend {blendOption=Linear}
    alphaScaleBiasClamp {bInterpResult=true}
    boneToModify {boneName=dyn_coat_01_l_MSTR}
    constraintSetup [{transformType=Rotation, targetBone {boneName=thigh_l}}]
    constraintWeights [1]
   }
   n94 = "Constraint - dyn_coat_01_r_MSTR (R)"(ComponentPose=@n93, Alpha=1, ConstraintWeights_0=1) {
    alphaBoolBlend {blendOption=Linear}
    boneToModify {boneName=dyn_coat_01_r_MSTR}
    constraintSetup [{transformType=Rotation, targetBone {boneName=thigh_r}}]
    constraintWeights [1]
   }
   n95 = "Constraint - dyn_coat_fr_0_r_MSTR (R)"(ComponentPose=@n94, Alpha=1, ConstraintWeights_0=1) {
    alphaBoolBlend {blendOption=Linear}
    boneToModify {boneName=dyn_coat_fr_0_r_MSTR}
    constraintSetup [{transformType=Rotation, targetBone {boneName=thigh_r}}]
    constraintWeights [1]
   }
   n96 = "Transform (Modify) Bone - Bone: dyn_coat_fr_0_l_MSTR"(ComponentPose=@n95, Alpha=1, Translation="0,0,0", Rotation="20.000,0.000,0.000", Scale="1,1,1") {
    rotationMode=BMM_Additive
    rotationSpace=BCS_BoneSpace
    alphaBoolBlend {blendOption=Linear}
    boneToModify {boneName=dyn_coat_fr_0_l_MSTR}
    rotation {pitch=20}
   }
   n97 = "Transform (Modify) Bone - Bone: dyn_coat_01_l_MSTR"(ComponentPose=@n96, Alpha=1, Translation="0,0,0", Rotation="20.000,0.000,0.000", Scale="1,1,1") {
    rotationMode=BMM_Additive
    rotationSpace=BCS_BoneSpace
    alphaBoolBlend {blendOption=Linear}
    boneToModify {boneName=dyn_coat_01_l_MSTR}
    rotation {pitch=20}
   }
   n98 = "Transform (Modify) Bone - Bone: dyn_coat_bk_01_l_MSTR"(ComponentPose=@n97, Alpha=1, Translation="0,0,0", Rotation="20.000,0.000,0.000", Scale="1,1,1") {
    rotationMode=BMM_Additive
    rotationSpace=BCS_BoneSpace
    alphaBoolBlend {blendOption=Linear}
    boneToModify {boneName=dyn_coat_bk_01_l_MSTR}
    rotation {pitch=20}
   }
   n99 = "Transform (Modify) Bone - Bone: dyn_coat_bk_01_MSTR"(ComponentPose=@n98, Alpha=1, Translation="0,0,0", Rotation="20.000,0.000,0.000", Scale="1,1,1") {
    rotationMode=BMM_Additive
    rotationSpace=BCS_BoneSpace
    alphaBoolBlend {blendOption=Linear}
    boneToModify {boneName=dyn_coat_bk_01_MSTR}
    rotation {pitch=20}
   }
   n100 = "Transform (Modify) Bone - Bone: dyn_coat_bk_01_r_MSTR"(ComponentPose=@n99, Alpha=1, Translation="0,0,0", Rotation="20.000,0.000,0.000", Scale="1,1,1") {
    rotationMode=BMM_Additive
    rotationSpace=BCS_BoneSpace
    alphaBoolBlend {blendOption=Linear}
    boneToModify {boneName=dyn_coat_bk_01_r_MSTR}
    rotation {pitch=20}
   }
   n101 = "Transform (Modify) Bone - Bone: dyn_coat_01_r_MSTR"(ComponentPose=@n100, Alpha=1, Translation="0,0,0", Rotation="20.000,0.000,0.000", Scale="1,1,1") {
    rotationMode=BMM_Additive
    rotationSpace=BCS_BoneSpace
    alphaBoolBlend {blendOption=Linear}
    boneToModify {boneName=dyn_coat_01_r_MSTR}
    rotation {pitch=20}
   }
   n102 = "Transform (Modify) Bone - Bone: dyn_coat_fr_0_r_MSTR"(ComponentPose=@n101, Alpha=1, Translation="0,0,0", Rotation="20.000,0.000,0.000", Scale="1,1,1") {
    rotationMode=BMM_Additive
    rotationSpace=BCS_BoneSpace
    alphaBoolBlend {blendOption=Linear}
    boneToModify {boneName=dyn_coat_fr_0_r_MSTR}
    rotation {pitch=20}
   }
   n103 = Get windVector() -> (windVector:Vector)
   n104 = RigidBody!(ComponentPose=@n102, Alpha, OverrideWorldGravity:Vector=@n103, ExternalForce) -> (Pose) {
    bClampLinearTranslationLimitToRefPose=true
    bOverrideWorldGravity=true
    overridePhysicsAsset=/BRCosmetics/Characters/Player/Male/Medium/Bodies/M_MED_CycloneSpace/Meshes/M_MED_Cyclone_Space_Hoodie_Physics
    simulationSpace=BaseBoneSpace
    alphaBoolBlend {blendOption=Linear}
    baseBoneRef {boneName=spine_04}
    componentAppliedLinearAccClamp=<3000 3000 3000>
    componentLinearAccScale=<0.3 0.3 0.3>
    componentLinearVelScale=<0.45 0.45 0.45>
    overrideWorldGravity {z=-980}
   }
   n105 = RigidBody(ComponentPose=@n104, OverrideWorldGravity=Get finalGravity() -> (finalGravity:Vector)) {
    bClampLinearTranslationLimitToRefPose=true
    bOverrideWorldGravity=true
    overridePhysicsAsset=/BRCosmetics/Characters/Player/Male/Medium/Bodies/M_MED_CycloneSpace/Meshes/M_MED_Cyclone_Space_Physics
    simulationSpace=BaseBoneSpace
    alphaBoolBlend {blendOption=Linear}
    baseBoneRef {boneName=root}
    componentAppliedLinearAccClamp=<1000 1000 1000>
    componentLinearAccScale=<0.3 0.3 0.1>
    componentLinearVelScale=<0.3 0.3 0.1>
    overrideWorldGravity {z=-980}
   }
   n106 = RigidBody(ComponentPose=@n105, OverrideWorldGravity=@n103) {
    bClampLinearTranslationLimitToRefPose=true
    bOverrideWorldGravity=true
    overridePhysicsAsset=/BRCosmetics/Characters/Player/Male/Medium/Bodies/M_MED_Cyclone/Meshes/M_MED_Cyclone_necklace_Physics
    simulationSpace=BaseBoneSpace
    alphaBoolBlend {blendOption=Linear}
    baseBoneRef {boneName=spine_05}
    componentAppliedLinearAccClamp=<3000 3000 3000>
    componentLinearAccScale=<0.3 0.3 0.3>
    componentLinearVelScale=<0.3 0.3 0.3>
    overrideWorldGravity {z=-980}
   }
   n107 = "Transform (Modify) Bone - Bone: dyn_necklace_01"(ComponentPose=@n106, Alpha=1, Translation="0,0,1", Rotation="0,0,0", Scale="1,1,1") {
    rotationMode=BMM_Additive
    rotationSpace=BCS_BoneSpace
    translationMode=BMM_Additive
    translationSpace=BCS_BoneSpace
    alphaBoolBlend {blendOption=Linear}
    boneToModify {boneName=dyn_necklace_01}
    translation {z=1}
   }
   n108 = "Transform (Modify) Bone - Bone: dyn_necklace_tp_01"(ComponentPose=@n107, Alpha=1, Translation="0,0,1", Rotation="0,0,0", Scale="1,1,1") {
    rotationMode=BMM_Additive
    rotationSpace=BCS_BoneSpace
    translationMode=BMM_Additive
    translationSpace=BCS_BoneSpace
    alphaBoolBlend {blendOption=Linear}
    boneToModify {boneName=dyn_necklace_tp_01}
    translation {z=1}
   }
   n109 = "Transform (Modify) Bone - Bone: dyn_necklace_tp_04"(ComponentPose=@n108, Alpha=1, Translation="0,0,0", Rotation="10.000,0.028,-0.012", Scale="1,1,1") {
    rotationMode=BMM_Additive
    rotationSpace=BCS_BoneSpace
    translationMode=BMM_Additive
    translationSpace=BCS_BoneSpace
    alphaBoolBlend {blendOption=Linear}
    boneToModify {boneName=dyn_necklace_tp_04}
    rotation=<rot 10 0.028 -0.012>
   }
   n110 = "Transform (Modify) Bone - Bone: dyn_necklace_03"(ComponentPose=@n109, Alpha=1, Translation="0,0,0", Rotation="10.000,-0.012,-0.022", Scale="1,1,1") {
    rotationMode=BMM_Additive
    rotationSpace=BCS_BoneSpace
    translationMode=BMM_Additive
    translationSpace=BCS_BoneSpace
    alphaBoolBlend {blendOption=Linear}
    boneToModify {boneName=dyn_necklace_03}
    rotation=<rot 10 -0.012 -0.022>
   }
   n111 = "Transform (Modify) Bone - Bone: dyn_necklace_tp_05"(ComponentPose=@n110, Alpha=1, Translation="0,0,0", Rotation="-5.000,-0.001,0.005", Scale="1,1,1") {
    rotationMode=BMM_Additive
    rotationSpace=BCS_BoneSpace
    translationMode=BMM_Additive
    translationSpace=BCS_BoneSpace
    alphaBoolBlend {blendOption=Linear}
    boneToModify {boneName=dyn_necklace_tp_05}
    rotation=<rot -5 -0.001 0.005>
   }
   n112 = "Transform (Modify) Bone - Bone: dyn_necklace_03"(ComponentPose=@n111, Alpha=Get isSkydiveAlpha() -> (isSkydiveAlpha:double), Translation="0,0,0", Rotation="15,0.000,0.000", Scale="1,1,1") {
    rotationMode=BMM_Additive
    rotationSpace=BCS_BoneSpace
    translationMode=BMM_Additive
    translationSpace=BCS_BoneSpace
    alphaBoolBlend {blendOption=Linear}
    boneToModify {boneName=dyn_necklace_03}
    rotation {pitch=15}
   }
   Save cached pose 'rbanCache'(Pose=Component To Local(ComponentPose=@n112)) {cacheName=rbanCache}
  ]}
 EventGraph {nodes [
   Event BlueprintUpdateAnimation()
   > n1 = Sequence!(execute) -> (then_0, then_1, then_2:Exec, then_3:Exec, then_4:Exec, then_5:Exec)
   n2 = Break Fort Anim Input Skydiving External Force(FortAnimInput_SkydivingExternalForce=Get SkydivingExternalForceInput()) -> (
    bUseSkydivingVectorForce=~
    SkydivingVectorForce=~
    HeadToPelvisDirection=~
    FloatingMultiplier=~
    FloatingAdditive=~
    DivingMultiplier=~
    DivingAdditive=~
    ParachutingMultiplier=~
    ParachutingAdditive=~
   )
   > n3 = Set windVector(windVector:Vector=SelectVector(A=vector * float(A=@n2.SkydivingVectorForce, B=0.05), B="0, 0,-980", bPickA=Get bIsSkyDiving())) -> (Output_Get:Vector)
   n4 = GetOwnerSkeletalMeshComponent()
   n5 = BreakRotator(InRot=BreakTransform(InTransform=GetDeltaTransformFromRefPose(self=@n4, BoneName=pelvis, BaseName=root)).Rotation)
   n6 = SelectFloat(A=980, B=-980, bPickA="float > float"(A="Absolute (Float)"(A=@n5.Yaw), B=90) {is=title K2Node_CallFunction})
   > Set finalGravity(finalGravity:Vector=SelectVector(A=MakeVector(Y=0, Z=@n6), B=@n3, bPickA=Get bIsPlayingEmote())) -> (Output_Get:Vector)
   Set isSkydiveAlpha(execute=@n1.then_1, isSkydiveAlpha:double="To Float (Boolean)"(InBool=Get bIsSkyDiving())) -> (Output_Get:double)
   Set rotCoat_0(execute=@n1.then_2, rotCoat_0:Rotator="0, 0, 0") -> (Output_Get:Rotator)
   > Set rotCoat_1(rotCoat_1:Rotator="0, 0, 0") -> (Output_Get:Rotator)
   > Set rotCoat_2(rotCoat_2:Rotator="1, 0, 0") -> (Output_Get:Rotator)
   > Set rotCoat_3(rotCoat_3:Rotator="2, 0, 0") -> (Output_Get:Rotator)
   > Set rotCoat_4(rotCoat_4:Rotator="4, 0, 0") -> (Output_Get:Rotator)
   > Set rotCoat_5(rotCoat_5:Rotator="8, 0, 0") -> (Output_Get:Rotator)
   > Set rotCoat_6(rotCoat_6:Rotator="16, 0, 0") -> (Output_Get:Rotator)
   > Set rotCoat_7(rotCoat_7:Rotator="35, 0, 0") -> (Output_Get:Rotator)
   n7 = Get CurrentLOD()
   n8 = Get bRigidBodyEnabled()
   Set rigids_off(execute=@n1.then_3, rigids_off:bool=OR Boolean(A=NOT Boolean(A=@n8), B="integer > integer"(A=@n7, B=1) {is=title K2Node_CallFunction})) -> (Output_Get:bool)
   n9 = Get PawnSpeed2D()
   n10 = "float >= float"(A=@n9, B=150) {is=title K2Node_CallFunction}
   n11 = Get bIsCrouching()
   Set lod2movingAndCrouchAlpha(
    execute=@n1.then_4
    lod2movingAndCrouchAlpha:double=SelectFloat(A=MapRangeClamped(Value=@n9, InRangeA=0, InRangeB=200), B=0, bPickA=AND Boolean(A=@n11, B=@n10))
   ) -> (Output_Get:double)
   Set isNotCrouchingAlpha(execute=@n1.then_5, isNotCrouchingAlpha:double=SelectFloat(A=0, B=1, bPickA=Get bIsCrouching())) -> (Output_Get:double)
   Event OnBeginCrouching()
   > Set crouchingAlpha(crouchingAlpha:double=0.75) -> (Output_Get:double)
   Event OnEndCrouching()
   > Set crouchingAlpha(crouchingAlpha=0.5)
   Event BlueprintInitializeAnimation()
   > Set crouchingAlpha(crouchingAlpha=0.5)
   n12 = SelectFloat(A=3, B=1, bPickA=Get bIsSkyDiving())
   n13 = MakeWaveRotator!(
    Amplitudes:Rotator="3,1,1"
    Frequencies:Rotator="Scale Rotator (Float)"!(A:Rotator="100, 0, 0", B:float=@n12) -> (ReturnValue:Rotator)
    Offsets:Rotator="100,0,0"
    bUseNoise:bool=true
    Min:Vector
    Max:Vector
   ) -> (ReturnValue:Rotator)
   n14 = Set noiseRot_0(noiseRot_0:Rotator=@n13) -> (Output_Get:Rotator)
   n15 = MakeWaveRotator(Amplitudes="-3,1,1", Frequencies="Scale Rotator (Float)"(A="125, 0, 0", B=@n12), Offsets="-85,0,0", bUseNoise=true)
   > n16 = Set noiseRot_1(noiseRot_1:Rotator="Scale Rotator (Float)"(A=@n15, B=1)) -> (Output_Get:Rotator)
   n17 = "Scale Rotator (Float)"(A="25, 0, 0", B=@n12)
   > n18 = Set noiseRot_2(noiseRot_2:Rotator="Scale Rotator (Float)"(A=MakeWaveRotator(Amplitudes="2.5,1,1", Frequencies=@n17, Offsets="75,0,0", bUseNoise=true, Min="0,-5, 0"), B=2.5)) -> (Output_Get:Rotator)
   > n19 = Set noiseRot_3(noiseRot_3:Rotator="Scale Rotator (Float)"(A=@n15, B=5)) -> (Output_Get:Rotator)
   > n20 = Set noiseRot_4(noiseRot_4:Rotator="Scale Rotator (Float)"(A=@n13, B=10)) -> (Output_Get:Rotator)
   > n21 = Set noiseRot_5(noiseRot_5:Rotator="Scale Rotator (Float)"(A=@n13, B=20)) -> (Output_Get:Rotator)
   > Set noiseRot_6(noiseRot_6:Rotator="Scale Rotator (Float)"(A=@n14, B=2)) -> (Output_Get:Rotator)
   > Set noiseRot_7(noiseRot_7:Rotator="Scale Rotator (Float)"(A=@n16, B=2)) -> (Output_Get:Rotator)
   > Set noiseRot_8(noiseRot_8:Rotator="Scale Rotator (Float)"(A=@n18, B=2)) -> (Output_Get:Rotator)
   > Set noiseRot_9(noiseRot_9:Rotator="Scale Rotator (Float)"(A=@n19, B=2)) -> (Output_Get:Rotator)
   > Set noiseRot_10(noiseRot_10:Rotator="Scale Rotator (Float)"(A=@n20, B=2)) -> (Output_Get:Rotator)
   > Set noiseRot_11(noiseRot_11:Rotator="Scale Rotator (Float)"(A=@n21, B=2)) -> (Output_Get:Rotator)
  ]}
]
variables [
 crouchingAlpha {type=number}
 finalGravity {type=Vector}
 isNotCrouchingAlpha {type=number}
 isSkydiveAlpha {type=number}
 lod2movingAndCrouchAlpha {type=number}
 noiseRot_0 {type=Rotator}
 noiseRot_1 {type=Rotator}
 noiseRot_10 {type=Rotator}
 noiseRot_11 {type=Rotator}
 noiseRot_2 {type=Rotator}
 noiseRot_3 {type=Rotator}
 noiseRot_4 {type=Rotator}
 noiseRot_5 {type=Rotator}
 noiseRot_6 {type=Rotator}
 noiseRot_7 {type=Rotator}
 noiseRot_8 {type=Rotator}
 noiseRot_9 {type=Rotator}
 rigids_off {type=boolean}
 rotCoat_0 {type=Rotator}
 rotCoat_1 {type=Rotator}
 rotCoat_2 {type=Rotator}
 rotCoat_3 {type=Rotator}
 rotCoat_4 {type=Rotator}
 rotCoat_5 {type=Rotator}
 rotCoat_6 {type=Rotator}
 rotCoat_7 {type=Rotator}
 windVector {type=Vector}
]
