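# Animation Blueprint for the Cyclone Universe face accessory, shrinking the globe bone while the character is targeting.
# It copies the pose from the attached parent mesh, then chains two Modify Bone nodes that replace globe scale with 0.9 and 1.0, with interpolated alphas.
# Targeting begin/end, customization-loaded and animation-init events set GlobeAlpha from bIsTargeting, then derive GlobeAlphaInv from it by subtraction.
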AnimBlueprint /BRCosmetics/Characters/Player/Male/Medium/Bodies/M_MED_CycloneUniverse/Meshes/Parts/M_MED_Cyclone_Universe_FaceAcc_AnimBP
target_skeleton=/BRCosmetics/Characters/Player/Male/Medium/Bodies/M_MED_Cyclone/Meshes/Parts/M_MED_Cyclone_FaceAcc_Skeleton
graphs [
 AnimGraph {nodes [
   n1 = "Transform (Modify) Bone - Bone: globe"(
    ComponentPose=Local To Component(LocalPose=Copy Pose From Mesh() {bUseAttachedParent=true})
    Alpha=Get GlobeAlpha() -> (GlobeAlpha:double)
    Translation="0,0,0"
    Rotation="0,0,0"
    Scale="0.9,0.9,0.9"
   ) {
    scaleMode=BMM_Replace
    scaleSpace=BCS_BoneSpace
    alphaBoolBlend {blendOption=Linear}
    alphaScaleBiasClamp {bInterpResult=true, interpSpeedDecreasing=15, interpSpeedIncreasing=15}
    boneToModify {boneName=globe}
    scale=<0.9 0.9 0.9>
   }
   n2 = "Transform (Modify) Bone - Bone: globe"(ComponentPose=@n1, Alpha=Get GlobeAlphaInv() -> (GlobeAlphaInv:double), Translation="0,0,0", Rotation="0,0,0", Scale="1,1,1") {
    scaleMode=BMM_Replace
    scaleSpace=BCS_BoneSpace
    alphaBoolBlend {blendOption=Linear}
    alphaScaleBiasClamp {bInterpResult=true, interpSpeedDecreasing=15, interpSpeedIncreasing=15}
    boneToModify {boneName=globe}
   }
   Output Pose(Result=Component To Local(ComponentPose=@n2)) {layerGroup=None}
  ]}
 EventGraph {nodes [
   n1 = Event OnBeginTargeting() {is=title K2Node_Event}
   n2 = Set GlobeAlpha(execute=[@n1.then, @n3.then, @n4.then, @n5.then], GlobeAlpha:double=SelectFloat(A=1, B=0, bPickA=Get bIsTargeting())) -> (Output_Get:double)
   > Set GlobeAlphaInv(GlobeAlphaInv:double=float - float(B=@n2) {is=title K2Node_CallFunction}) -> (Output_Get:double)
   n3 = Event OnEndTargeting() {is=title K2Node_Event}
   n4 = Event OnCharacterCustomizationLoaded()
   n5 = Event BlueprintInitializeAnimation()
  ]}
]
variables [
 GlobeAlpha {type=number}
 GlobeAlphaInv {type=number}
]
